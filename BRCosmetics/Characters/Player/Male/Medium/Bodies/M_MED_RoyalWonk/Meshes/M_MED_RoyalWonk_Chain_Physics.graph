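# Physics asset for the RoyalWonk medium male body's watch chain, letting it swing against the character's torso and leg.
# Three dynamic chain bodies, two of capsules and a box at the end, link back to a base body through angle-limited constraints, with the first joint allowing no twist.
# Unconstrained capsules on pelvis, spine, left thigh and RB_Helper bones for inner leg, belly flap, pelvis and spine shape the collision surface around it.
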PhysicsAsset /BRCosmetics/Characters/Player/Male/Medium/Bodies/M_MED_RoyalWonk/Meshes/M_MED_RoyalWonk_Chain_Physics
bodies {
 InnerLeg_RB_Helper [{
   shapeType=Capsule
   length=2.3327
   radius=13.1195
   shapeName=thigh_twist_01_l_capsule
   center=<0.9404 -12.7875 -0.4023>
   rotation=<rot 85.6399 -170.5238 158.1242>
  }]
 InnerLeg_RB_Helper_base [{shapeType=Capsule, length=0.505, radius=0.505, shapeName=InnerLeg_RB_Helper_base_capsule, rotation=<rot 0 -90 0>}]
 base_WatchChain_01 [{shapeType=Capsule, length=0.505, radius=0.505, shapeName=base_WatchChain_01_capsule, rotation=<rot 0 -90 0>}]
 bellyJacketFlap_RB_Helper [{
   shapeType=Capsule
   length=2.0189
   radius=1.8411
   shapeName=bellyJacketFlap_RB_Helper_capsule
   center=<0.6658 -1.9517 0.3007>
   rotation=<rot -1.9636 -92.7624 54.8119>
  }]
 dyn_WatchChain_01 [{
   shapeType=Capsule
   length=14.3469
   radius=0.8015
   shapeName=dyn_WatchChain_01_capsule
   center=<4.0248 -0.2863 0.4983>
   rotation=<rot 3.1236 86.3968 3.31>
  }]
 dyn_WatchChain_02 [
  {
   shapeType=Capsule
   length=12.7717
   radius=0.8015
   shapeName=dyn_WatchChain_01_capsule
   center=<4.0248 -0.2863 0.4983>
   rotation=<rot 3.1237 86.3968 3.31>
  }
  {
   shapeType=Capsule
   length=4.4101
   radius=0.8015
   shapeName=dyn_WatchChain_01_capsule
   center=<2.17 -0.0961 -3.7523>
   rotation=<rot 3.1237 86.3968 3.31>
  }
 ]
 dyn_WatchChain_03 [{shapeType=Box, extentX=5.7541, extentY=1.1779, extentZ=12.4988, shapeName=dyn_WatchChain_03_box, center=<3.5374 0 0.6655>}]
 pelvis [{
   shapeType=Capsule
   length=12.334
   radius=19.5897
   shapeName=pelvis_capsule
   center=<-3.5364 -0.5358 0.0358>
   rotation=<rot 0.0398 -90 0.5678>
  }]
 pelvis_RB_Helper [{
   shapeType=Capsule
   length=3.5966
   radius=12.1637
   shapeName=pelvis_RB_Helper_capsule
   center=<-2.7162 -12.1585 0.8349>
   rotation=<rot -0.0001 -134.973 90.0001>
  }]
 pelvis_RB_Helper_base [{shapeType=Capsule, length=0.0877, radius=0.5, shapeName=pelvis_RB_Helper_base_capsule, rotation=<rot 0 -90 0>}]
 spine_01 [{shapeType=Capsule, length=4.6519, radius=18.8683, shapeName=spine_01_capsule, rotation=<rot 0 -90 0>}]
 spine_02 [{shapeType=Capsule, length=4.6519, radius=16.7332, shapeName=spine_01_capsule, center=<0 -0.2775 1.5858>, rotation=<rot 0 -90 0>}]
 spine_02_RB_Helper [{
   shapeType=Capsule
   length=3.5966
   radius=12.1637
   shapeName=pelvis_RB_Helper_capsule
   center=<-3.7777 -10.172 -6.2214>
   rotation=<rot -0.0001 -134.973 90.0001>
  }]
 spine_02_RB_Helper_base [{shapeType=Capsule, length=0.505, radius=0.505, shapeName=spine_02_RB_Helper_base_capsule, rotation=<rot 0 -90 0>}]
 spine_03 [{shapeType=Capsule, length=5.3973, radius=13.5859, shapeName=spine_01_capsule, center=<0 -0.7903 -1.7618>, rotation=<rot 44.6229 -90.0002 0>}]
 spine_04 [{
   shapeType=Capsule
   length=7.119
   radius=5.6879
   shapeName=spine_04_capsule
   center=<6.7011 11.2188 -0.0254>
   rotation=<rot 0.0521 -90 -4.2939>
  }]
 thigh_twist_01_l [{
   shapeType=Capsule
   length=19.8134
   radius=13.6163
   shapeName=thigh_twist_01_l_capsule
   center=<1.5083 1.4064 -4.9128>
   rotation=<rot -85.7874 -179.9987 -175.1692>
  }]
}
constraints [
 {bone1Name=dyn_WatchChain_01, bone2Name=base_WatchChain_01, swing1LimitDegrees=30, swing2LimitDegrees=2, twistLimitDegrees=0}
 {bone1Name=dyn_WatchChain_02, bone2Name=dyn_WatchChain_01, swing1LimitDegrees=30, swing2LimitDegrees=15, twistLimitDegrees=60}
 {bone1Name=dyn_WatchChain_03, bone2Name=dyn_WatchChain_02, swing1LimitDegrees=30, swing2LimitDegrees=20, twistLimitDegrees=30}
]
