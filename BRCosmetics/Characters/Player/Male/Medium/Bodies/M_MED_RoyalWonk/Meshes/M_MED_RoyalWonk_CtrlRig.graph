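# Control Rig for the RoyalWonk male medium body cosmetic, correcting jacket flaps, watch chain, suit back flaps and leg deformation bones.
# Weighted parent constraints anchor accessory and helper bones to spine, clavicle, pelvis and thigh bones.
# Quaternion-delta alphas and spherical pose readers on spine, arms and legs weight additive transform correctives.
# Pelvis-height crouch estimates, spine bend/twist, smoothed crouch/targeting flags and an AnimDynamics toggle gate some correctives.
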ControlRigBlueprint /BRCosmetics/Characters/Player/Male/Medium/Bodies/M_MED_RoyalWonk/Meshes/M_MED_RoyalWonk_CtrlRig
graphs [
 CollapseNode_ContainedGraph {nodes [
   n1 = RigVMFunctionEntryNode()
   &RigUnit_ParentConstraint(
    AdvancedSettings=(InterpolationType=Average,RotationOrderForFilter=XZY)
    Child=(Type=Bone,Name="base_jacketFlap_top_01_l")
    Filter=(TranslationFilter=(bX=True,bY=True,bZ=True),RotationFilter=(bX=true,bY=true,bZ=true),ScaleFilter=(bX=True,bY=True,bZ=True))
    Parents="((Item=(Type=Bone,Name=\"spine_03\"),Weight=0.026),(Item=(Type=Bone,Name=\"spine_04\"),Weight=0.059),(Item=(Type=Bone,Name=\"spine_05\"),Weight=0.46),(Item=(Type=Bone,Name=\"clavicle_l\"),Weight=0.398),(Item=(Type=Bone,Name=\"neck_02\"),Weight=0.019),(Item=(Type=Bon ...(truncated)"
    Weight=1
    ExecutePin=@n1.ExecutePin
   )
   > &RigUnit_RotationConstraintLocalSpaceOffset(
    AdvancedSettings=(InterpolationType=Average,RotationOrderForFilter=XZY)
    Child=(Type=Bone,Name="base_jacketFlap_top_01_l")
    Filter=(bX=True,bY=True,bZ=True)
    Parents=((Item=(Type=Bone,Name="clavicle_l"),Weight=1))
    Weight=1
    bMaintainOffset=true
   )
   > &RigUnit_ParentConstraint(
    AdvancedSettings=(InterpolationType=Average,RotationOrderForFilter=XZY)
    Child=(Type=Bone,Name="base_jacketFlap_top_01_r")
    Filter=(TranslationFilter=(bX=True,bY=True,bZ=True),RotationFilter=(bX=true,bY=true,bZ=true),ScaleFilter=(bX=True,bY=True,bZ=True))
    Parents="((Item=(Type=Bone,Name=\"spine_03\"),Weight=0.026),(Item=(Type=Bone,Name=\"spine_04\"),Weight=0.059),(Item=(Type=Bone,Name=\"spine_05\"),Weight=0.46),(Item=(Type=Bone,Name=\"clavicle_r\"),Weight=0.398),(Item=(Type=Bone,Name=\"neck_02\"),Weight=0.019),(Item=(Type=Bon ...(truncated)"
    Weight=1
   )
   > &RigUnit_RotationConstraintLocalSpaceOffset(
    AdvancedSettings=(InterpolationType=Average,RotationOrderForFilter=XZY)
    Child=(Type=Bone,Name="base_jacketFlap_top_01_r")
    Filter=(bX=True,bY=True,bZ=True)
    Parents=((Item=(Type=Bone,Name="clavicle_r"),Weight=1))
    Weight=1
   )
   > &RigUnit_ParentConstraint(
    AdvancedSettings=(InterpolationType=Average,RotationOrderForFilter=XZY)
    Child=(Type=Bone,Name="base_jacketFlap_low_01_l")
    Filter=(TranslationFilter=(bX=True,bY=True,bZ=True),RotationFilter=(bX=true,bY=true,bZ=true),ScaleFilter=(bX=True,bY=True,bZ=True))
    Parents="((Item=(Type=Bone,Name=\"spine_03\"),Weight=0.204),(Item=(Type=Bone,Name=\"spine_04\"),Weight=0.383),(Item=(Type=Bone,Name=\"spine_05\"),Weight=0.165),(Item=(Type=Bone,Name=\"clavicle_l\"),Weight=0.057),(Item=(Type=Bone,Name=\"spine_02\"),Weight=0.093),(Item=(Type=B ...(truncated)"
    Weight=1
   )
   > &RigUnit_ParentConstraint(
    AdvancedSettings=(InterpolationType=Average,RotationOrderForFilter=XZY)
    Child=(Type=Bone,Name="base_jacketFlap_row_01_r")
    Filter=(TranslationFilter=(bX=True,bY=True,bZ=True),RotationFilter=(bX=true,bY=true,bZ=true),ScaleFilter=(bX=True,bY=True,bZ=True))
    Parents="((Item=(Type=Bone,Name=\"spine_03\"),Weight=0.204),(Item=(Type=Bone,Name=\"spine_04\"),Weight=0.383),(Item=(Type=Bone,Name=\"spine_05\"),Weight=0.165),(Item=(Type=Bone,Name=\"clavicle_r\"),Weight=0.057),(Item=(Type=Bone,Name=\"spine_02\"),Weight=0.093),(Item=(Type=B ...(truncated)"
    Weight=1
   )
   > &RigUnit_ParentConstraint(
    AdvancedSettings=(InterpolationType=Average,RotationOrderForFilter=XZY)
    Child=(Type=Bone,Name="base_WatchChain_01")
    Filter=(TranslationFilter=(bX=True,bY=True,bZ=True),RotationFilter=(bX=true,bY=true,bZ=true),ScaleFilter=(bX=True,bY=True,bZ=True))
    Parents="((Item=(Type=Bone,Name=\"spine_03\"),Weight=0.372),(Item=(Type=Bone,Name=\"spine_04\"),Weight=0.177),(Item=(Type=Bone,Name=\"spine_05\"),Weight=0.165),(Item=(Type=Bone,Name=\"spine_01\"),Weight=0.077),(Item=(Type=Bone,Name=\"spine_02\"),Weight=0.202),(Item=(Type=Bon ...(truncated)"
    Weight=1
   )
   > &RigUnit_ParentConstraint(
    AdvancedSettings=(InterpolationType=Average,RotationOrderForFilter=XZY)
    Child=(Type=Bone,Name="base_SuitBackFlap_01_l")
    Filter=(TranslationFilter=(bX=True,bY=True,bZ=True),RotationFilter=(bX=true,bY=true,bZ=true),ScaleFilter=(bX=True,bY=True,bZ=True))
    Parents=((Item=(Type=Bone,Name="spine_03"),Weight=0.22),(Item=(Type=Bone,Name="spine_04"),Weight=0.04),(Item=(Type=Bone,Name="spine_01"),Weight=0.2),(Item=(Type=Bone,Name="spine_02"),Weight=0.3),(Item=(Type=Bone,Name="pelvis"),Weight=0.24))
    Weight=1
   )
   > &RigUnit_ParentConstraint(
    AdvancedSettings=(InterpolationType=Average,RotationOrderForFilter=XZY)
    Child=(Type=Bone,Name="base_SuitBackFlap_01_r")
    Filter=(TranslationFilter=(bX=True,bY=True,bZ=True),RotationFilter=(bX=true,bY=true,bZ=true),ScaleFilter=(bX=True,bY=True,bZ=True))
    Parents=((Item=(Type=Bone,Name="spine_03"),Weight=0.22),(Item=(Type=Bone,Name="spine_04"),Weight=0.04),(Item=(Type=Bone,Name="spine_01"),Weight=0.2),(Item=(Type=Bone,Name="spine_02"),Weight=0.3),(Item=(Type=Bone,Name="pelvis"),Weight=0.24))
    Weight=1
   )
   > &RigUnit_ParentConstraint(
    AdvancedSettings=(RotationOrderForFilter=XZY)
    Child=(Type=Bone,Name="spine_02_RB_Helper_base")
    Parents=((Item=(Type=Bone,Name="spine_04"),Weight=0.0225),(Item=(Type=Bone,Name="spine_03"),Weight=0.1883),(Item=(Type=Bone,Name="spine_01"),Weight=0.1921),(Item=(Type=Bone,Name="spine_02"),Weight=0.2716),(Item=(Type=Bone,Name="pelvis"),Weight=0.3254))
    Weight=1
   )
   > &RigUnit_ParentConstraint(
    AdvancedSettings=(RotationOrderForFilter=XZY)
    Child=(Type=Bone,Name="pelvis_RB_Helper_base")
    Parents="((Item=(Type=Bone,Name=\"spine_03\"),Weight=0.0289),(Item=(Type=Bone,Name=\"spine_01\"),Weight=0.0418),(Item=(Type=Bone,Name=\"spine_02\"),Weight=0.0503),(Item=(Type=Bone,Name=\"thigh_twist_01_l\"),Weight=0.0698),(Item=(Type=Bone,Name=\"deform_groin_l\"),Weight=0.15 ...(truncated)"
    Weight=1
   )
   > &RigUnit_ParentConstraint(
    AdvancedSettings=(RotationOrderForFilter=XZY)
    Child=(Type=Bone,Name="InnerLeg_RB_Helper")
    Parents="((Item=(Type=Bone,Name=\"thigh_twist_01_r\"),Weight=0.0091),(Item=(Type=Bone,Name=\"thigh_l\"),Weight=0.0483),(Item=(Type=Bone,Name=\"pelvis\"),Weight=0.123),(Item=(Type=Bone,Name=\"deform_groin_l\"),Weight=0.3163),(Item=(Type=Bone,Name=\"thigh_twist_01_l\"),Weight= ...(truncated)"
    Weight=1
   )
   > n2 = &RigUnit_ParentConstraint(
    AdvancedSettings=(RotationOrderForFilter=XZY)
    Child=(Type=Bone,Name="bellyJacketFlap_RB_Helper_base")
    Parents=((Item=(Type=Bone,Name="spine_04"),Weight=0.0479),(Item=(Type=Bone,Name="pelvis"),Weight=0.1436),(Item=(Type=Bone,Name="spine_01"),Weight=0.1895),(Item=(Type=Bone,Name="spine_03"),Weight=0.2817),(Item=(Type=Bone,Name="spine_02"),Weight=0.3374))
    Weight=1
   )
   RigVMFunctionReturnNode(ExecutePin=@n2.ExecutePin)
  ]}
 CollapseNode_ContainedGraph {nodes [
   RigVMFunctionEntryNode()
   > RigVMFunction_Sequence = RigVMAggregateNode()
   GetAlphaFromQuatDelta_6 = RigVMFunctionReferenceNode(
    Rotation Order=ZYX
    child=(Type=Bone,Name="spine_05")
    debugRotation=false
    outputMinMax=(X=0,Y=1)
    parent=(Type=Null,Name="spine_05_null")
    premult=1
    rotationMinMax=(X=0,Y=40)
    triggerAxis=Z
    twistAxis=(X=1,Y=0,Z=0)
    ExecuteContext=@RigVMFunction_Sequence.A
   )
   > &RigUnit_ModifyTransforms(
    ItemToModify="((Item=(Type=Bone,Name=\"base_WatchChain_01\"),Transform=(Rotation=(X=0,Y=0,Z=0.0872,W=0.9962),Translation=(X=0,Y=1.5,Z=0),Scale3D=(X=1,Y=1,Z=1))),(Item=(Type=Bone,Name=\"dyn_WatchChain_02\"),Transform=(Rotation=(X=0.0436,Y=0,Z=0,W=0.999),Translation=(X=0,Y=0, ...(truncated)"
    WeightMinimum=0
    Weight=@GetAlphaFromQuatDelta_6.alpha
   )
   > GetAlphaFromQuatDelta_10 = RigVMFunctionReferenceNode(
    Rotation Order=ZYX
    child=(Type=Bone,Name="spine_04")
    debugRotation=false
    outputMinMax=(X=0,Y=1)
    parent=(Type=Null,Name="spine_04_null")
    premult=1
    rotationMinMax=(X=0,Y=40)
    triggerAxis=Z
    twistAxis=(X=1,Y=0,Z=0)
   )
   > &RigUnit_ModifyTransforms(
    ItemToModify="((Item=(Type=Bone,Name=\"base_WatchChain_01\"),Transform=(Rotation=(X=0,Y=0,Z=0.0872,W=0.9962),Translation=(X=0,Y=2,Z=0),Scale3D=(X=1,Y=1,Z=1))),(Item=(Type=Bone,Name=\"dyn_WatchChain_02\"),Transform=(Rotation=(X=0.0436,Y=0,Z=0,W=0.999),Translation=(X=0,Y=1,Z= ...(truncated)"
    WeightMinimum=0
    Weight=@GetAlphaFromQuatDelta_10.alpha
   )
   > GetAlphaFromQuatDelta_11 = RigVMFunctionReferenceNode(
    Rotation Order=ZYX
    child=(Type=Bone,Name="spine_03")
    debugRotation=false
    outputMinMax=(X=0,Y=1)
    parent=(Type=Null,Name="spine_03_null")
    premult=1
    rotationMinMax=(X=0,Y=40)
    triggerAxis=Z
    twistAxis=(X=1,Y=0,Z=0)
   )
   > &RigUnit_ModifyTransforms(
    ItemToModify="((Item=(Type=Bone,Name=\"base_WatchChain_01\"),Transform=(Rotation=(X=0,Y=0,Z=0.1736,W=0.9848),Translation=(X=0,Y=1,Z=0),Scale3D=(X=1,Y=1,Z=1))),(Item=(Type=Bone,Name=\"dyn_WatchChain_02\"),Transform=(Rotation=(X=0,Y=0,Z=0,W=1),Translation=(X=0,Y=0,Z=0),Scale3 ...(truncated)"
    WeightMinimum=0
    Weight=@GetAlphaFromQuatDelta_11.alpha
   )
   > GetAlphaFromQuatDelta_12 = RigVMFunctionReferenceNode(
    Rotation Order=ZYX
    child=(Type=Bone,Name="spine_02")
    debugRotation=false
    outputMinMax=(X=0,Y=1)
    parent=(Type=Null,Name="spine_02_null")
    premult=1
    rotationMinMax=(X=0,Y=40)
    triggerAxis=Z
    twistAxis=(X=1,Y=0,Z=0)
   )
   > &RigUnit_ModifyTransforms(
    ItemToModify="((Item=(Type=Bone,Name=\"base_WatchChain_01\"),Transform=(Rotation=(X=0,Y=0,Z=0.1045,W=0.9945),Translation=(X=0,Y=0,Z=0),Scale3D=(X=1,Y=1,Z=1))),(Item=(Type=Bone,Name=\"dyn_WatchChain_02\"),Transform=(Rotation=(X=0,Y=0,Z=0.0872,W=0.9962),Translation=(X=0,Y=0,Z ...(truncated)"
    WeightMinimum=0
    Weight=@GetAlphaFromQuatDelta_12.alpha
   )
   > GetAlphaFromQuatDelta_7 = RigVMFunctionReferenceNode(
    Rotation Order=ZYX
    child=(Type=Bone,Name="spine_01")
    debugRotation=false
    outputMinMax=(X=0,Y=1)
    parent=(Type=Null,Name="spine_01_null")
    premult=1
    rotationMinMax=(X=0,Y=40)
    triggerAxis=Z
    twistAxis=(X=1,Y=0,Z=0)
   )
   > RigVMVariableNode(Value=@GetAlphaFromQuatDelta_7.alpha) {variable=Spine01_ForwardBend}
   > &RigUnit_ModifyTransforms(
    ItemToModify="((Item=(Type=Bone,Name=\"base_WatchChain_01\"),Transform=(Rotation=(X=0,Y=0,Z=0.1132,W=0.9936),Translation=(X=0,Y=0,Z=0),Scale3D=(X=1,Y=1,Z=1))),(Item=(Type=Bone,Name=\"dyn_WatchChain_02\"),Transform=(Rotation=(X=0,Y=0,Z=0.0872,W=0.9962),Translation=(X=0,Y=0,Z ...(truncated)"
    WeightMinimum=0
    Weight=@GetAlphaFromQuatDelta_7.alpha
   )
   > GetAlphaFromQuatDelta_9 = RigVMFunctionReferenceNode(
    Rotation Order=ZYX
    child=(Type=Bone,Name="spine_04")
    debugRotation=false
    onlyTwist=true
    outputMinMax=(X=0,Y=1)
    parent=(Type=Null,Name="spine_05_DeltaSpine1_null")
    premult=1
    rotationMinMax=(X=0,Y=-35)
    triggerAxis=X
    twistAxis=(X=1,Y=0,Z=0)
   )
   > &RigUnit_ModifyTransforms(
    ItemToModify="((Item=(Type=Bone,Name=\"base_WatchChain_01\"),Transform=(Rotation=(X=-0.0654,Y=-0.0017,Z=0.0261,W=0.9975),Translation=(X=0,Y=0.5,Z=-1.5),Scale3D=(X=1,Y=1,Z=1))),(Item=(Type=Bone,Name=\"dyn_WatchChain_02\"),Transform=(Rotation=(X=0.0436,Y=0,Z=0,W=0.999),Transl ...(truncated)"
    WeightMinimum=0
    Weight=@GetAlphaFromQuatDelta_9.alpha
   )
   > GetAlphaFromQuatDelta_16 = RigVMFunctionReferenceNode(
    Rotation Order=ZYX
    child=(Type=Bone,Name="spine_04")
    debugRotation=false
    onlyTwist=true
    outputMinMax=(X=1,Y=0.7)
    parent=(Type=Null,Name="spine_05_DeltaSpine1_null")
    premult=1
    rotationMinMax=(X=0,Y=20)
    triggerAxis=X
    twistAxis=(X=1,Y=0,Z=0)
   )
   n1 = &RigVMFunction_AlphaInterp(
    InterpSpeedDecreasing=8
    InterpSpeedIncreasing=8
    Scale=1
    bClampResult=False
    Value=&RigVMFunction_MathBoolToFloat(Value=RigVMVariableNode() {variable=isCrouchingb}.Value).Result
   )
   Multiply_8 = RigVMAggregateNode(B=@n1.Result, C=RigVMVariableNode() {variable=spine03To01BendBackwards}.Value, A=@GetAlphaFromQuatDelta_16.alpha)
   > n2 = &RigUnit_ModifyTransforms(
    ItemToModify="((Item=(Type=Bone,Name=\"base_WatchChain_01\"),Transform=(Rotation=(X=0,Y=0,Z=-0.1045,W=0.9945),Translation=(X=0,Y=0,Z=0),Scale3D=(X=1,Y=1,Z=1))),(Item=(Type=Bone,Name=\"dyn_WatchChain_02\"),Transform=(Rotation=(X=0.0941,Y=0.0789,Z=-0.0941,W=0.988),Translation ...(truncated)"
    WeightMinimum=0
    Weight=@Multiply_8.Result
   )
   RigVMFunction_ControlFlowBranch = &RigVMFunction_ControlFlowBranch(ExecuteContext=@n2.ExecutePin, Condition=RigVMVariableNode() {variable=AnimDynamicsEnabled}.Value)
   RigVMFunction_MathBoolNot = &RigVMFunction_MathBoolNot(Value=RigVMVariableNode() {variable=isCrouchingb}.Value)
   n3 = &RigVMFunction_AlphaInterp(
    InterpSpeedDecreasing=8
    InterpSpeedIncreasing=8
    Scale=1
    bClampResult=False
    Value=&RigVMFunction_MathBoolToFloat(Value=@RigVMFunction_MathBoolNot.Result).Result
   )
   Multiply_9 = RigVMAggregateNode(B=@n3.Result, C=RigVMVariableNode() {variable=spine03To01BendBackwards}.Value, A=@GetAlphaFromQuatDelta_16.alpha)
   &RigUnit_ModifyTransforms(
    ItemToModify="((Item=(Type=Bone,Name=\"base_WatchChain_01\"),Transform=(Rotation=(X=0,Y=0,Z=0,W=1),Translation=(X=0,Y=0,Z=0),Scale3D=(X=1,Y=1,Z=1))),(Item=(Type=Bone,Name=\"dyn_WatchChain_02\"),Transform=(Rotation=(X=-0.0163,Y=0.0441,Z=-0.0254,W=0.9986),Translation=(X=0,Y=0 ...(truncated)"
    WeightMinimum=0
    Weight=@Multiply_9.Result
    ExecutePin=@RigVMFunction_ControlFlowBranch.False
   )
   GetAlphaFromQuatDelta_2 = RigVMFunctionReferenceNode(
    Rotation Order=ZYX
    child=(Type=Bone,Name="spine_04")
    debugRotation=false
    onlyTwist=true
    outputMinMax=(X=1,Y=0)
    parent=(Type=Null,Name="spine_05_DeltaSpine1_null")
    premult=1
    rotationMinMax=(X=0,Y=15)
    triggerAxis=X
    twistAxis=(X=1,Y=0,Z=0)
    ExecuteContext=@RigVMFunction_ControlFlowBranch.True
   )
   > &RigUnit_ModifyTransforms(
    ItemToModify="((Item=(Type=Bone,Name=\"base_WatchChain_01\"),Transform=(Rotation=(X=0,Y=0,Z=0,W=1),Translation=(X=0,Y=-0.5,Z=0),Scale3D=(X=1,Y=1,Z=1))),(Item=(Type=Bone,Name=\"dyn_WatchChain_02\"),Transform=(Rotation=(X=-0.0163,Y=0.0441,Z=-0.0254,W=0.9986),Translation=(X=0, ...(truncated)"
    WeightMinimum=0
    Weight=@GetAlphaFromQuatDelta_2.alpha
   )
   RigVMFunctionReturnNode(ExecuteContext=@RigVMFunction_Sequence.C)
   GetAlphaFromQuatDelta_8 = RigVMFunctionReferenceNode(
    Rotation Order=ZYX
    child=(Type=Bone,Name="thigh_l")
    debugRotation=false
    outputMinMax=(X=0,Y=1)
    parent=(Type=Null,Name="thigh_l_null")
    premult=1
    rotationMinMax=(X=-90,Y=-115)
    triggerAxis=Z
    twistAxis=(X=1,Y=0,Z=0)
    ExecuteContext=@RigVMFunction_Sequence.B
   )
   > RigVMVariableNode(Value=@GetAlphaFromQuatDelta_8.alpha) {variable=standing_L_ChainLegRaise}
   n4 = RigVMVariableNode() {variable=isCrouchingb}
   n5 = &RigVMFunction_AlphaInterp(
    InterpSpeedDecreasing=10
    InterpSpeedIncreasing=10
    Scale=1
    bClampResult=False
    Value=&RigVMFunction_MathBoolToFloat(Value=&RigVMFunction_MathBoolNot(Value=@n4.Value).Result).Result
   )
   n6 = &RigVMFunction_AnimEvalRichCurve(
    Curve=(EditorCurveData=(Keys=((InterpMode=RCIM_Cubic,TangentMode=RCTM_User,ArriveTangent=1.1834,LeaveTangent=1.1834),(InterpMode=RCIM_Cubic,TangentMode=RCTM_User,Time=1,Value=1,ArriveTangent=0.6328,LeaveTangent=0.6328))))
    SourceMinimum=0
    TargetMinimum=0
    Value=@GetAlphaFromQuatDelta_8.alpha
   )
   > &RigUnit_ModifyTransforms(
    ItemToModify="((Item=(Type=Bone,Name=\"base_WatchChain_01\"),Transform=(Rotation=(X=0,Y=0,Z=0.1478,W=0.989),Translation=(X=0,Y=0,Z=0),Scale3D=(X=1,Y=1,Z=1))),(Item=(Type=Bone,Name=\"dyn_WatchChain_02\"),Transform=(Rotation=(X=-0.1816,Y=0.1711,Z=0.294,W=0.9227),Translation=( ...(truncated)"
    WeightMinimum=0
    Weight=&RigVMFunction_MathDoubleMul(B=@n5.Result, A=@n6.Result).Result
   )
   > GetAlphaFromQuatDelta_13 = RigVMFunctionReferenceNode(
    Rotation Order=ZYX
    child=(Type=Bone,Name="thigh_l")
    debugRotation=false
    outputMinMax=(X=0,Y=1)
    parent=(Type=Null,Name="thigh_l_null")
    premult=1
    rotationMinMax=(X=-60,Y=-90)
    triggerAxis=Z
    twistAxis=(X=1,Y=0,Z=0)
   )
   Remap = &RigVMFunction_MathDoubleRemap(SourceMaximum=0.8, SourceMinimum=0.3, TargetMaximum=1, TargetMinimum=0, Value=RigVMVariableNode() {variable=Spine01_ForwardBend}.Value)
   > &RigUnit_ModifyTransforms(
    ItemToModify="((Item=(Type=Bone,Name=\"base_WatchChain_01\"),Transform=(Rotation=(X=0,Y=0,Z=0.2588,W=0.9659),Translation=(X=0,Y=0,Z=0),Scale3D=(X=1,Y=1,Z=1))),(Item=(Type=Bone,Name=\"dyn_WatchChain_02\"),Transform=(Rotation=(X=-0.1561,Y=0.1948,Z=0.1627,W=0.9546),Translation ...(truncated)"
    WeightMinimum=0
    Weight=&RigVMFunction_MathDoubleMul(A=@GetAlphaFromQuatDelta_13.alpha, B=@Remap.Result).Result
   )
   > GetAlphaFromQuatDelta_14 = RigVMFunctionReferenceNode(
    Rotation Order=ZYX
    child=(Type=Bone,Name="thigh_l")
    debugRotation=false
    outputMinMax=(X=0,Y=1)
    parent=(Type=Null,Name="thigh_l_null")
    premult=1
    rotationMinMax=(X=-60,Y=-80)
    triggerAxis=Z
    twistAxis=(X=1,Y=0,Z=0)
   )
   n7 = RigVMVariableNode() {variable=isTargetting}
   n8 = RigVMVariableNode() {variable=isCrouchingb}
   n9 = &RigVMFunction_AlphaInterp(
    InterpSpeedDecreasing=8
    InterpSpeedIncreasing=8
    Scale=1
    bClampResult=False
    Value=&RigVMFunction_MathBoolToFloat(Value=&RigVMFunction_MathBoolAnd(B=@n7.Value, A=@n8.Value).Result).Result
   )
   n10 = &RigVMFunction_AnimEvalRichCurve(
    Curve=(EditorCurveData=(Keys=((InterpMode=RCIM_Cubic,TangentMode=RCTM_User,ArriveTangent=1.1834,LeaveTangent=1.1834),(InterpMode=RCIM_Cubic,TangentMode=RCTM_User,Time=1,Value=1,ArriveTangent=0.6328,LeaveTangent=0.6328))))
    SourceMinimum=0
    TargetMinimum=0
    Value=@GetAlphaFromQuatDelta_14.alpha
   )
   Multiply_5 = RigVMAggregateNode(
    B=@n9.Result
    A=@n10.Result
    C=&RigVMFunction_MathDoubleSub(A=1, B=RigVMVariableNode() {variable=spineTwistingRight}.Value).Result
   )
   n11 = RigVMVariableNode() {variable=fullSpineBend}
   n12 = &RigVMFunction_AlphaInterp(
    InterpSpeedDecreasing=7
    InterpSpeedIncreasing=14
    Scale=1
    bClampResult=False
    Value=&RigVMFunction_MathDoubleMul(
     A=@Multiply_5.Result
     B=&RigVMFunction_MathDoubleRemap(SourceMaximum=1, SourceMinimum=0, TargetMaximum=1, TargetMinimum=0.3, bClamp=False, Value=@n11.Value).Result
    ).Result
   )
   > &RigUnit_ModifyTransforms(
    ItemToModify="((Item=(Type=Bone,Name=\"base_WatchChain_01\"),Transform=(Rotation=(X=0,Y=0,Z=0.1908,W=0.9816),Translation=(X=0,Y=0.5,Z=-3),Scale3D=(X=1,Y=1,Z=1))),(Item=(Type=Bone,Name=\"dyn_WatchChain_02\"),Transform=(Rotation=(X=-0.1875,Y=0.1976,Z=0.1212,W=0.9545),Translat ...(truncated)"
    WeightMinimum=0
    Weight=@n12.Result
   )
   > GetAlphaFromQuatDelta_1 = RigVMFunctionReferenceNode(
    Rotation Order=ZYX
    child=(Type=Bone,Name="thigh_l")
    debugRotation=false
    onlyTwist=true
    outputMinMax=(X=0,Y=1)
    parent=(Type=Null,Name="thigh_l_null")
    premult=1
    rotationMinMax=(X=-34,Y=-5)
    triggerAxis=Z
    twistAxis=(X=1,Y=0,Z=0)
   )
   Multiply = RigVMAggregateNode(B=1, C=@n12.Result, A=@GetAlphaFromQuatDelta_1.alpha)
   > &RigUnit_ModifyTransforms(
    ItemToModify="((Item=(Type=Bone,Name=\"base_WatchChain_01\"),Transform=(Rotation=(X=0,Y=0,Z=0.0436,W=0.999),Translation=(X=0,Y=0.5,Z=-2),Scale3D=(X=1,Y=1,Z=1))),(Item=(Type=Bone,Name=\"dyn_WatchChain_02\"),Transform=(Rotation=(X=-0.2162,Y=-0.0094,Z=0.0426,W=0.9754),Translat ...(truncated)"
    WeightMinimum=0
    Weight=@Multiply.Result
   )
   > GetAlphaFromQuatDelta_15 = RigVMFunctionReferenceNode(
    Rotation Order=ZYX
    child=(Type=Bone,Name="thigh_l")
    debugRotation=false
    outputMinMax=(X=0,Y=1)
    parent=(Type=Null,Name="thigh_l_null")
    premult=1
    rotationMinMax=(X=-60,Y=-90)
    triggerAxis=Z
    twistAxis=(X=1,Y=0,Z=0)
   )
   n13 = &RigVMFunction_MathBoolNot(Value=RigVMVariableNode() {variable=isTargetting}.Value)
   n14 = RigVMVariableNode() {variable=isCrouchingb}
   n15 = &RigVMFunction_AlphaInterp(
    InterpSpeedDecreasing=8
    InterpSpeedIncreasing=8
    Scale=1
    bClampResult=False
    Value=&RigVMFunction_MathBoolToFloat(Value=&RigVMFunction_MathBoolAnd(B=@n13.Result, A=@n14.Value).Result).Result
   )
   n16 = &RigVMFunction_AnimEvalRichCurve(
    Curve=(EditorCurveData=(Keys=((InterpMode=RCIM_Cubic,TangentMode=RCTM_User,ArriveTangent=1.3988,LeaveTangent=1.3988),(InterpMode=RCIM_Cubic,TangentMode=RCTM_User,Time=1,Value=1,ArriveTangent=0.6328,LeaveTangent=0.6328))))
    SourceMinimum=0
    TargetMinimum=0
    Value=@GetAlphaFromQuatDelta_15.alpha
   )
   Multiply_6 = RigVMAggregateNode(
    B=@n15.Result
    A=@n16.Result
    C=&RigVMFunction_MathDoubleSub(A=1, B=RigVMVariableNode() {variable=spineTwistingRight}.Value).Result
   )
   > &RigUnit_ModifyTransforms(
    ItemToModify="((Item=(Type=Bone,Name=\"base_WatchChain_01\"),Transform=(Rotation=(X=0,Y=0,Z=0.1305,W=0.9914),Translation=(X=0,Y=0.5,Z=0),Scale3D=(X=1,Y=1,Z=1))),(Item=(Type=Bone,Name=\"dyn_WatchChain_02\"),Transform=(Rotation=(X=-0.2685,Y=0.0381,Z=0.1893,W=0.9437),Translati ...(truncated)"
    WeightMinimum=0
    Weight=@Multiply_6.Result
   )
   RigVMFunction_MathBoolToFloat = &RigVMFunction_MathBoolToFloat(Value=&RigVMFunction_MathBoolNot(Value=RigVMVariableNode() {variable=isTargetting}.Value).Result)
   &RigVMFunction_AlphaInterp(InterpSpeedDecreasing=8, InterpSpeedIncreasing=8, Scale=1, bClampResult=False, Value=@RigVMFunction_MathBoolToFloat.Result)
   &RigVMFunction_AlphaInterp(InterpSpeedDecreasing=10, InterpSpeedIncreasing=20, Scale=1, bClampResult=False, Value=@Multiply_6.Result)
  ]}
 CollapseNode_ContainedGraph {nodes [
   RigVMFunctionEntryNode()
   n1 = &RigUnit_GetRelativeTransformForItem(Child=(Type=Bone,Name="pelvis"), Parent=(Type=Bone,Name="root"), bChildInitial=False, bParentInitial=False)
   > RigVMVariableNode(Value=&RigVMFunction_AnimEvalRichCurve(
     Curve=(EditorCurveData=(Keys=((),(Time=1,Value=1))))
     SourceMinimum=0
     TargetMinimum=0
     Value=&RigVMFunction_MathDoubleRemap(SourceMaximum=85, SourceMinimum=44, TargetMaximum=0, TargetMinimum=1, bClamp=false, Value=@n1.RelativeTransform.Translation.Z).Result
    ).Result) {variable=isCrouchingStill}
   n2 = &RigVMFunction_AnimEvalRichCurve(
    Curve=(EditorCurveData=(Keys=((InterpMode=RCIM_Cubic),(InterpMode=RCIM_Cubic,Time=0.5,Value=1),(Time=0.75))))
    SourceMaximum=2
    SourceMinimum=0
    TargetMinimum=0
    Value=&RigVMFunction_MathDoubleRemap(
     SourceMaximum=75
     SourceMinimum=45
     TargetMaximum=1
     TargetMinimum=0
     bClamp=false
     Value=&RigUnit_GetRelativeTransformForItem(Child=(Type=Bone,Name="pelvis"), Parent=(Type=Bone,Name="root"), bChildInitial=False, bParentInitial=False).RelativeTransform.Translation.Z
    ).Result
   )
   > RigVMVariableNode(Value=@n2.Result) {variable=isCrouchWalking}
   > GetAlphaFromQuatDelta_7 = RigVMFunctionReferenceNode(
    Rotation Order=ZYX
    child=(Type=Bone,Name="spine_04")
    debugRotation=false
    onlyTwist=true
    outputMinMax=(X=0,Y=1)
    parent=(Type=Null,Name="spine_05_DeltaSpine1_null")
    premult=1
    rotationMinMax=(X=-3,Y=24)
    triggerAxis=X
    twistAxis=(X=1,Y=0,Z=0)
   )
   > RigVMVariableNode(Value=@GetAlphaFromQuatDelta_7.alpha) {variable=spineTwistingRight}
   > GetAlphaFromQuatDelta_1 = RigVMFunctionReferenceNode(
    Rotation Order=ZYX
    child=(Type=Bone,Name="spine_03")
    debugRotation=false
    onlyTwist=false
    outputMinMax=(X=1,Y=0)
    parent=(Type=Null,Name="spine_03_DeltaSpine01_null")
    premult=1
    rotationMinMax=(X=0,Y=15)
    triggerAxis=Z
    twistAxis=(X=1,Y=0,Z=0)
   )
   > RigVMVariableNode(Value=@GetAlphaFromQuatDelta_1.alpha) {variable=spine03To01BendBackwards}
   > GetAlphaFromQuatDelta_17 = RigVMFunctionReferenceNode(
    Rotation Order=ZYX
    child=(Type=Bone,Name="spine_05")
    debugRotation=false
    outputMinMax=(X=0,Y=1)
    parent=(Type=Null,Name="spine_05_DeltaSpine1_null")
    premult=1
    rotationMinMax=(X=20,Y=38)
    triggerAxis=Z
    twistAxis=(X=1,Y=0,Z=0)
   )
   > RigVMVariableNode(Value=@GetAlphaFromQuatDelta_17.alpha) {variable=fullSpineBend}
   > RigVMFunctionReturnNode()
  ]}
 CollapseNode_ContainedGraph {nodes [
   RigVMFunctionEntryNode()
   > GetAlphaFromQuatDelta_1 = RigVMFunctionReferenceNode(
    Rotation Order=ZYX
    child=(Type=Bone,Name="calf_l")
    debugRotation=false
    outputMinMax=(X=0,Y=1)
    parent=(Type=Null,Name="calf_l_null")
    premult=1
    rotationMinMax=(X=85,Y=140)
    triggerAxis=Z
    twistAxis=(X=1,Y=0,Z=0)
   )
   > &RigUnit_ModifyTransforms(
    ItemToModify="((Item=(Type=Bone,Name=\"quadBack_hi_l\"),Transform=(Rotation=(X=-0.1305,Y=0,Z=0,W=0.9914),Translation=(X=0.05,Y=10,Z=1.1),Scale3D=(X=1,Y=1,Z=1))),(Item=(Type=Bone,Name=\"quadBack_low_l\"),Transform=(Rotation=(X=0,Y=0,Z=0,W=1),Translation=(X=2.7,Y=9.5,Z=0),Sca ...(truncated)"
    WeightMinimum=0
    Weight=@GetAlphaFromQuatDelta_1.alpha
   )
   > GetAlphaFromQuatDelta_2 = RigVMFunctionReferenceNode(
    Rotation Order=ZYX
    child=(Type=Bone,Name="calf_r")
    debugRotation=false
    outputMinMax=(X=0,Y=1)
    parent=(Type=Null,Name="calf_r_null")
    premult=1
    rotationMinMax=(X=85,Y=140)
    triggerAxis=Z
    twistAxis=(X=1,Y=0,Z=0)
   )
   > &RigUnit_ModifyTransforms(
    ItemToModify="((Item=(Type=Bone,Name=\"quadBack_hi_r\"),Transform=(Rotation=(X=-0.1305,Y=0,Z=0,W=0.9914),Translation=(X=-0.05,Y=-10,Z=-1.1),Scale3D=(X=1,Y=1,Z=1))),(Item=(Type=Bone,Name=\"quadBack_low_r\"),Transform=(Rotation=(X=0,Y=0,Z=0,W=1),Translation=(X=-2.7,Y=-9.5,Z=0 ...(truncated)"
    WeightMinimum=0
    Weight=@GetAlphaFromQuatDelta_2.alpha
   )
   > RigVMFunctionReturnNode()
  ]}
 CollapseNode_ContainedGraph {nodes [
   RigVMFunctionEntryNode()
   > RigVMFunction_Sequence = RigVMAggregateNode()
   GetAlphaFromQuatDelta = RigVMFunctionReferenceNode(
    Rotation Order=ZYX
    child=(Type=Bone,Name="upperarm_r")
    debugRotation=false
    outputMinMax=(X=0,Y=1)
    parent=(Type=Null,Name="upperarm_r_null")
    premult=1
    rotationMinMax=(X=60,Y=100)
    triggerAxis=Z
    twistAxis=(X=1,Y=0,Z=0)
    ExecuteContext=@RigVMFunction_Sequence.B
   )
   AlphaInterp = &RigVMFunction_AlphaInterp(InterpSpeedDecreasing=11, InterpSpeedIncreasing=15, Scale=1, bClampResult=False, Value=@GetAlphaFromQuatDelta.alpha)
   > &RigUnit_ModifyTransforms(
    ItemToModify="((Item=(Type=Bone,Name=\"base_jacketFlap_top_01_r\"),Transform=(Rotation=(X=-0.2317,Y=0.3149,Z=0.1537,W=0.9075),Translation=(X=-1,Y=0,Z=-1),Scale3D=(X=0.8,Y=1,Z=1))),(Item=(Type=Bone,Name=\"dyn_jacketFlap_top_02_r\"),Transform=(Rotation=(X=0,Y=0,Z=0,W=1),Trans ...(truncated)"
    WeightMinimum=0
    Weight=@AlphaInterp.Result
   )
   n1 = &RigVMFunction_AlphaInterp(
    InterpSpeedDecreasing=11
    InterpSpeedIncreasing=15
    Scale=1
    bClampResult=False
    Value=&RigVMFunction_MathBoolToFloat(Value=RigVMVariableNode() {variable=isCrouchingb}.Value).Result
   )
   > &RigUnit_ModifyTransforms(
    ItemToModify="((Item=(Type=Bone,Name=\"base_jacketFlap_top_01_r\"),Transform=(Rotation=(X=-0.0189,Y=0.2156,Z=0.0851,W=0.9726),Translation=(X=0,Y=0,Z=0),Scale3D=(X=1,Y=1,Z=1))),(Item=(Type=Bone,Name=\"dyn_jacketFlap_top_02_r\"),Transform=(Rotation=(X=0,Y=0.0872,Z=0,W=0.9962) ...(truncated)"
    WeightMinimum=0
    Weight=&RigVMFunction_MathDoubleMul(B=@n1.Result, A=@AlphaInterp.Result).Result
   )
   > n2 = &RigUnit_SphericalPoseReader(
    ActiveRegionScaleFactors=(PositiveWidth=1,NegativeWidth=1,PositiveHeight=1,NegativeHeight=1)
    ActiveRegionSize=0.05
    Debug=(bDrawDebug=false,bDraw2D=False,bDrawLocalAxes=False,DebugScale=25,DebugSegments=20,DebugThickness=0.25)
    DriverAxis=(X=-1,Y=0,Z=0)
    DriverItem=(Type=Bone,Name="upperarm_r")
    FalloffRegionScaleFactors=(PositiveWidth=1,NegativeWidth=1,PositiveHeight=1,NegativeHeight=1)
    FalloffSize=0.2
    FlipHeightScaling=False
    FlipWidthScaling=False
    OptionalParentItem=(Type=Bone,Name="spine_05")
    RotationOffset=(X=0,Y=50,Z=45)
   )
   RigVMFunction_MathBoolToFloat = &RigVMFunction_MathBoolToFloat(Value=RigVMVariableNode() {variable=isTargetting}.Value)
   n3 = &RigVMFunction_AlphaInterp(InterpSpeedDecreasing=11, InterpSpeedIncreasing=15, Scale=1, bClampResult=False, Value=@RigVMFunction_MathBoolToFloat.Result)
   Multiply = &RigVMFunction_MathDoubleMul(B=@n3.Result, A=@n2.OutputParam)
   > &RigUnit_ModifyTransforms(
    ItemToModify="((Item=(Type=Bone,Name=\"base_jacketFlap_top_01_r\"),Transform=(Rotation=(X=0,Y=0,Z=0,W=1),Translation=(X=0,Y=0,Z=0),Scale3D=(X=1,Y=1,Z=1))),(Item=(Type=Bone,Name=\"dyn_jacketFlap_row_01_r\"),Transform=(Rotation=(X=0.1294,Y=-0.1294,Z=0.017,W=0.983),Translation ...(truncated)"
    WeightMinimum=0
    Weight=@Multiply.Result
   )
   > n4 = &RigUnit_SphericalPoseReader(
    ActiveRegionScaleFactors=(PositiveWidth=0.5,NegativeWidth=0.5,PositiveHeight=0.5,NegativeHeight=0.5)
    ActiveRegionSize=0.1
    Debug=(bDrawDebug=false,bDraw2D=False,bDrawLocalAxes=False,DebugScale=25,DebugSegments=20,DebugThickness=0.25)
    DriverAxis=(X=-1,Y=0,Z=0)
    DriverItem=(Type=Bone,Name="upperarm_r")
    FalloffRegionScaleFactors=(PositiveWidth=1,NegativeWidth=1,PositiveHeight=1,NegativeHeight=1)
    FalloffSize=0.2
    FlipHeightScaling=False
    FlipWidthScaling=False
    OptionalParentItem=(Type=Bone,Name="clavicle_r")
    RotationOffset=(X=120,Y=0,Z=-145)
   )
   > &RigUnit_ModifyTransforms(
    ItemToModify="((Item=(Type=Bone,Name=\"base_jacketFlap_top_01_r\"),Transform=(Rotation=(X=0,Y=0.2164,Z=0,W=0.9763),Translation=(X=1.5,Y=0,Z=0.5),Scale3D=(X=1,Y=1,Z=1))),(Item=(Type=Bone,Name=\"dyn_jacketFlap_top_02_r\"),Transform=(Rotation=(X=0,Y=0.1736,Z=0,W=0.9848),Transl ...(truncated)"
    WeightMinimum=0
    Weight=@n4.OutputParam
   )
   Remap = &RigVMFunction_MathDoubleRemap(SourceMaximum=0.5, SourceMinimum=0, TargetMaximum=1, TargetMinimum=0, bClamp=False, Value=@n4.OutputParam)
   n5 = &RigVMFunction_AlphaInterp(
    InterpSpeedDecreasing=11
    InterpSpeedIncreasing=15
    Scale=1
    bClampResult=False
    Value=&RigVMFunction_MathBoolToFloat(Value=RigVMVariableNode() {variable=isCrouchingb}.Value).Result
   )
   Multiply_2 = RigVMAggregateNode(A=@Remap.Result, C=RigVMVariableNode() {variable=isCrouchWalking}.Value, B=@n5.Result)
   > &RigUnit_ModifyTransforms(
    ItemToModify=((Item=(Type=Bone,Name="base_jacketFlap_top_01_r"),Transform=(Rotation=(X=0.0868,Y=0.0868,Z=-0.0076,W=0.9924),Translation=(X=0,Y=0,Z=-1.5),Scale3D=(X=1,Y=1,Z=1))))
    WeightMinimum=0
    Weight=@Multiply_2.Result
   )
   > GetAlphaFromQuatDelta_3 = RigVMFunctionReferenceNode(
    Rotation Order=ZYX
    child=(Type=Bone,Name="clavicle_r")
    debugRotation=false
    outputMinMax=(X=0,Y=1)
    parent=(Type=Null,Name="clavicle_r_null")
    premult=1
    rotationMinMax=(X=0,Y=15)
    triggerAxis=Z
    twistAxis=(X=1,Y=0,Z=0)
   )
   > &RigUnit_ModifyTransforms(
    ItemToModify=((Item=(Type=Bone,Name="base_jacketFlap_row_01_r"),Transform=(Rotation=(X=0.0868,Y=0.0868,Z=-0.0076,W=0.9924),Translation=(X=0,Y=0,Z=0.5),Scale3D=(X=1,Y=1,Z=1))))
    WeightMinimum=0
    Weight=@GetAlphaFromQuatDelta_3.alpha
   )
   GetAlphaFromQuatDelta_1 = RigVMFunctionReferenceNode(
    Rotation Order=ZYX
    child=(Type=Bone,Name="upperarm_l")
    debugRotation=false
    outputMinMax=(X=0,Y=1)
    parent=(Type=Null,Name="upperarm_l_null")
    premult=1
    rotationMinMax=(X=70,Y=100)
    triggerAxis=Z
    twistAxis=(X=1,Y=0,Z=0)
    ExecuteContext=@RigVMFunction_Sequence.C
   )
   n6 = &RigVMFunction_AlphaInterp(InterpSpeedDecreasing=11, InterpSpeedIncreasing=15, Scale=1, bClampResult=False, Value=@GetAlphaFromQuatDelta_1.alpha)
   > &RigUnit_ModifyTransforms(
    ItemToModify="((Item=(Type=Bone,Name=\"base_jacketFlap_top_01_l\"),Transform=(Rotation=(X=0.2317,Y=0.3149,Z=-0.1537,W=0.9075),Translation=(X=-1,Y=0,Z=-1),Scale3D=(X=0.8,Y=1,Z=1))),(Item=(Type=Bone,Name=\"dyn_jacketFlap_top_02_l\"),Transform=(Rotation=(X=0,Y=0,Z=0,W=1),Trans ...(truncated)"
    WeightMinimum=0
    Weight=@n6.Result
   )
   n7 = &RigVMFunction_AlphaInterp(
    InterpSpeedDecreasing=11
    InterpSpeedIncreasing=15
    Scale=1
    bClampResult=False
    Value=&RigVMFunction_MathBoolToFloat(Value=RigVMVariableNode() {variable=isCrouchingb}.Value).Result
   )
   > &RigUnit_ModifyTransforms(
    ItemToModify="((Item=(Type=Bone,Name=\"base_jacketFlap_top_01_l\"),Transform=(Rotation=(X=-0.0189,Y=0.2156,Z=0.0851,W=0.9726),Translation=(X=0,Y=0,Z=0),Scale3D=(X=1,Y=1,Z=1))),(Item=(Type=Bone,Name=\"dyn_jacketFlap_top_02_l\"),Transform=(Rotation=(X=0,Y=0.0872,Z=0,W=0.9962) ...(truncated)"
    WeightMinimum=0
    Weight=&RigVMFunction_MathDoubleMul(B=@n7.Result, A=@n6.Result).Result
   )
   > n8 = &RigUnit_SphericalPoseReader(
    ActiveRegionScaleFactors=(PositiveWidth=1,NegativeWidth=1,PositiveHeight=1,NegativeHeight=1)
    ActiveRegionSize=0.05
    Debug=(bDrawDebug=false,bDraw2D=False,bDrawLocalAxes=False,DebugScale=25,DebugSegments=20,DebugThickness=0.25)
    DriverAxis=(X=-1,Y=0,Z=0)
    DriverItem=(Type=Bone,Name="upperarm_l")
    FalloffRegionScaleFactors=(PositiveWidth=1,NegativeWidth=1,PositiveHeight=1,NegativeHeight=1)
    FalloffSize=0.2
    FlipHeightScaling=False
    FlipWidthScaling=False
    OptionalParentItem=(Type=Bone,Name="spine_05")
    RotationOffset=(X=0,Y=50,Z=45)
   )
   n9 = &RigVMFunction_AlphaInterp(
    InterpSpeedDecreasing=11
    InterpSpeedIncreasing=15
    Scale=1
    bClampResult=False
    Value=&RigVMFunction_MathBoolToFloat(Value=RigVMVariableNode() {variable=isTargetting}.Value).Result
   )
   > &RigUnit_ModifyTransforms(
    ItemToModify="((Item=(Type=Bone,Name=\"base_jacketFlap_top_01_l\"),Transform=(Rotation=(X=0,Y=0,Z=0,W=1),Translation=(X=0,Y=0,Z=0),Scale3D=(X=1,Y=1,Z=1))),(Item=(Type=Bone,Name=\"dyn_jacketFlap_low_01_l\"),Transform=(Rotation=(X=-0.171,Y=-0.171,Z=-0.0302,W=0.9698),Translati ...(truncated)"
    WeightMinimum=0
    Weight=&RigVMFunction_MathDoubleMul(B=@n9.Result, A=@n8.OutputParam).Result
   )
   > n10 = &RigUnit_SphericalPoseReader(
    ActiveRegionScaleFactors=(PositiveWidth=1,NegativeWidth=1,PositiveHeight=1,NegativeHeight=1)
    ActiveRegionSize=0.05
    Debug=(bDrawDebug=false,bDraw2D=False,bDrawLocalAxes=False,DebugScale=25,DebugSegments=20,DebugThickness=0.25)
    DriverAxis=(X=1,Y=0,Z=0)
    DriverItem=(Type=Bone,Name="upperarm_l")
    FalloffRegionScaleFactors=(PositiveWidth=1,NegativeWidth=1,PositiveHeight=1,NegativeHeight=1)
    FalloffSize=0.2
    FlipHeightScaling=False
    FlipWidthScaling=False
    OptionalParentItem=(Type=Bone,Name="spine_05")
    RotationOffset=(X=0,Y=-120,Z=45)
   )
   > &RigUnit_ModifyTransforms(
    ItemToModify="((Item=(Type=Bone,Name=\"base_jacketFlap_top_01_l\"),Transform=(Rotation=(X=0,Y=0.2164,Z=0,W=0.9763),Translation=(X=1.5,Y=0,Z=0.5),Scale3D=(X=1,Y=1,Z=1))),(Item=(Type=Bone,Name=\"dyn_jacketFlap_top_02_l\"),Transform=(Rotation=(X=0,Y=0.1736,Z=0,W=0.9848),Transl ...(truncated)"
    WeightMinimum=0
    Weight=@n10.OutputParam
   )
   n11 = &RigVMFunction_AlphaInterp(
    InterpSpeedDecreasing=11
    InterpSpeedIncreasing=15
    Scale=1
    bClampResult=False
    Value=&RigVMFunction_MathBoolToFloat(Value=RigVMVariableNode() {variable=isCrouchingb}.Value).Result
   )
   > &RigUnit_ModifyTransforms(
    ItemToModify="((Item=(Type=Bone,Name=\"base_jacketFlap_top_01_r\"),Transform=(Rotation=(X=0,Y=0.1736,Z=0,W=0.9848),Translation=(X=1.5,Y=0,Z=0.5),Scale3D=(X=1,Y=1,Z=1))),(Item=(Type=Bone,Name=\"dyn_jacketFlap_top_02_r\"),Transform=(Rotation=(X=0,Y=0.0872,Z=0,W=0.9962),Transl ...(truncated)"
    WeightMinimum=0
    Weight=&RigVMFunction_MathDoubleMul(B=@n11.Result, A=@n10.OutputParam).Result
   )
   > GetAlphaFromQuatDelta_2 = RigVMFunctionReferenceNode(
    Rotation Order=ZYX
    child=(Type=Bone,Name="clavicle_l")
    debugRotation=false
    outputMinMax=(X=0,Y=1)
    parent=(Type=Null,Name="clavicle_l_null")
    premult=1
    rotationMinMax=(X=0,Y=15)
    triggerAxis=Z
    twistAxis=(X=1,Y=0,Z=0)
   )
   > &RigUnit_ModifyTransforms(
    ItemToModify=((Item=(Type=Bone,Name="base_jacketFlap_low_01_l"),Transform=(Rotation=(X=-0.0868,Y=0.0868,Z=0.0076,W=0.9924),Translation=(X=0,Y=0,Z=0.5),Scale3D=(X=1,Y=1,Z=1))))
    WeightMinimum=0
    Weight=@GetAlphaFromQuatDelta_2.alpha
   )
   RigVMFunctionReturnNode(ExecuteContext=@RigVMFunction_Sequence.D)
   n12 = &RigVMFunction_AlphaInterp(
    InterpSpeedDecreasing=11
    InterpSpeedIncreasing=15
    Scale=1
    bClampResult=False
    Value=&RigVMFunction_MathBoolToFloat(Value=RigVMVariableNode() {variable=isCrouchingb}.Value).Result
   )
   &RigUnit_ModifyTransforms(
    ItemToModify=((Item=(Type=Bone,Name="dyn_jacketFlap_top_01_r"),Transform=(Rotation=(X=0,Y=0,Z=0.0436,W=0.999),Translation=(X=0,Y=1,Z=0),Scale3D=(X=1,Y=1,Z=1))))
    WeightMinimum=0
    Weight=@n12.Result
    ExecutePin=@RigVMFunction_Sequence.A
   )
  ]}
 CollapseNode_ContainedGraph {nodes [
   RigVMFunctionEntryNode()
   > RigVMFunction_Sequence = RigVMAggregateNode()
   &RigUnit_ModifyTransforms(
    ItemToModify="((Item=(Type=Bone,Name=\"dyn_WatchChain_02_in\"),Transform=(Rotation=(X=0,Y=0,Z=0,W=1),Translation=(X=0,Y=0,Z=0),Scale3D=(X=1,Y=1,Z=1))),(Item=(Type=Bone,Name=\"base_WatchChain_01\"),Transform=(Rotation=(X=-0.0179,Y=0.017,Z=0.0265,W=0.9993),Translation=(X=0,Y= ...(truncated)"
    Weight=1
    WeightMinimum=0
    ExecutePin=@RigVMFunction_Sequence.A
   )
   GetAlphaFromQuatDelta_10 = RigVMFunctionReferenceNode(
    Rotation Order=ZYX
    child=(Type=Bone,Name="spine_04")
    debugRotation=false
    outputMinMax=(X=0,Y=1)
    parent=(Type=Null,Name="spine_04_null")
    premult=1
    rotationMinMax=(X=0,Y=40)
    triggerAxis=Z
    twistAxis=(X=1,Y=0,Z=0)
    ExecuteContext=@RigVMFunction_Sequence.B
   )
   > &RigUnit_ModifyTransforms(
    ItemToModify="((Item=(Type=Bone,Name=\"base_WatchChain_01\"),Transform=(Rotation=(X=0,Y=0,Z=0.0872,W=0.9962),Translation=(X=0,Y=2,Z=0),Scale3D=(X=1,Y=1,Z=1))),(Item=(Type=Bone,Name=\"dyn_WatchChain_02\"),Transform=(Rotation=(X=0.0436,Y=0,Z=0,W=0.999),Translation=(X=0,Y=1,Z= ...(truncated)"
    WeightMinimum=0
    Weight=@GetAlphaFromQuatDelta_10.alpha
   )
   > GetAlphaFromQuatDelta_11 = RigVMFunctionReferenceNode(
    Rotation Order=ZYX
    child=(Type=Bone,Name="spine_03")
    debugRotation=false
    outputMinMax=(X=0,Y=1)
    parent=(Type=Null,Name="spine_03_null")
    premult=1
    rotationMinMax=(X=0,Y=40)
    triggerAxis=Z
    twistAxis=(X=1,Y=0,Z=0)
   )
   > &RigUnit_ModifyTransforms(
    ItemToModify="((Item=(Type=Bone,Name=\"base_WatchChain_01\"),Transform=(Rotation=(X=0,Y=0,Z=0.1736,W=0.9848),Translation=(X=0,Y=1,Z=0),Scale3D=(X=1,Y=1,Z=1))),(Item=(Type=Bone,Name=\"dyn_WatchChain_02\"),Transform=(Rotation=(X=0,Y=0,Z=0,W=1),Translation=(X=0,Y=0,Z=0),Scale3 ...(truncated)"
    WeightMinimum=0
    Weight=@GetAlphaFromQuatDelta_11.alpha
   )
   > GetAlphaFromQuatDelta_9 = RigVMFunctionReferenceNode(
    Rotation Order=ZYX
    child=(Type=Bone,Name="spine_04")
    debugRotation=false
    onlyTwist=true
    outputMinMax=(X=0,Y=1)
    parent=(Type=Null,Name="spine_05_DeltaSpine1_null")
    premult=1
    rotationMinMax=(X=0,Y=-35)
    triggerAxis=X
    twistAxis=(X=1,Y=0,Z=0)
   )
   > &RigUnit_ModifyTransforms(
    ItemToModify="((Item=(Type=Bone,Name=\"base_WatchChain_01\"),Transform=(Rotation=(X=-0.0654,Y=-0.0017,Z=0.0261,W=0.9975),Translation=(X=0,Y=0.5,Z=-1.5),Scale3D=(X=1,Y=1,Z=1))),(Item=(Type=Bone,Name=\"dyn_WatchChain_02\"),Transform=(Rotation=(X=0.0436,Y=0,Z=0,W=0.999),Transl ...(truncated)"
    WeightMinimum=0
    Weight=@GetAlphaFromQuatDelta_9.alpha
   )
   GetAlphaFromQuatDelta_8 = RigVMFunctionReferenceNode(
    Rotation Order=ZYX
    child=(Type=Bone,Name="thigh_l")
    debugRotation=false
    outputMinMax=(X=0,Y=1)
    parent=(Type=Null,Name="thigh_l_null")
    premult=1
    rotationMinMax=(X=-90,Y=-115)
    triggerAxis=Z
    twistAxis=(X=1,Y=0,Z=0)
    ExecuteContext=@RigVMFunction_Sequence.C
   )
   n1 = RigVMVariableNode() {variable=isCrouchingb}
   n2 = &RigVMFunction_AlphaInterp(
    InterpSpeedDecreasing=10
    InterpSpeedIncreasing=10
    Scale=1
    bClampResult=False
    Value=&RigVMFunction_MathBoolToFloat(Value=&RigVMFunction_MathBoolNot(Value=@n1.Value).Result).Result
   )
   n3 = &RigVMFunction_AnimEvalRichCurve(
    Curve=(EditorCurveData=(Keys=((InterpMode=RCIM_Cubic,TangentMode=RCTM_User,ArriveTangent=1.1834,LeaveTangent=1.1834),(InterpMode=RCIM_Cubic,TangentMode=RCTM_User,Time=1,Value=1,ArriveTangent=0.6328,LeaveTangent=0.6328))))
    SourceMinimum=0
    TargetMinimum=0
    Value=@GetAlphaFromQuatDelta_8.alpha
   )
   > &RigUnit_ModifyTransforms(
    ItemToModify=((Item=(Type=Bone,Name="base_WatchChain_01"),Transform=(Rotation=(X=0,Y=0,Z=0.1736,W=0.9848),Translation=(X=0,Y=0,Z=0),Scale3D=(X=1,Y=1,Z=1))))
    WeightMinimum=0
    Weight=&RigVMFunction_MathDoubleMul(B=@n2.Result, A=@n3.Result).Result
   )
   > GetAlphaFromQuatDelta_14 = RigVMFunctionReferenceNode(
    Rotation Order=ZYX
    child=(Type=Bone,Name="thigh_l")
    debugRotation=false
    outputMinMax=(X=0,Y=1)
    parent=(Type=Null,Name="thigh_l_null")
    premult=1
    rotationMinMax=(X=-60,Y=-80)
    triggerAxis=Z
    twistAxis=(X=1,Y=0,Z=0)
   )
   n4 = &RigVMFunction_AlphaInterp(
    InterpSpeedDecreasing=8
    InterpSpeedIncreasing=8
    Scale=1
    bClampResult=False
    Value=&RigVMFunction_MathBoolToFloat(Value=RigVMVariableNode() {variable=isCrouchingb}.Value).Result
   )
   n5 = &RigVMFunction_AnimEvalRichCurve(
    Curve=(EditorCurveData=(Keys=((InterpMode=RCIM_Cubic,TangentMode=RCTM_User,ArriveTangent=1.1834,LeaveTangent=1.1834),(InterpMode=RCIM_Cubic,TangentMode=RCTM_User,Time=1,Value=1,ArriveTangent=0.6328,LeaveTangent=0.6328))))
    SourceMinimum=0
    TargetMinimum=0
    Value=@GetAlphaFromQuatDelta_14.alpha
   )
   Multiply_5 = RigVMAggregateNode(
    B=@n4.Result
    A=@n5.Result
    C=&RigVMFunction_MathDoubleSub(A=1, B=RigVMVariableNode() {variable=spineTwistingRight}.Value).Result
   )
   n6 = RigVMVariableNode() {variable=fullSpineBend}
   n7 = &RigVMFunction_AlphaInterp(
    InterpSpeedDecreasing=7
    InterpSpeedIncreasing=14
    Scale=1
    bClampResult=False
    Value=&RigVMFunction_MathDoubleMul(
     A=@Multiply_5.Result
     B=&RigVMFunction_MathDoubleRemap(SourceMaximum=1, SourceMinimum=0, TargetMaximum=1, TargetMinimum=0.3, bClamp=False, Value=@n6.Value).Result
    ).Result
   )
   > &RigUnit_ModifyTransforms(
    ItemToModify=((Item=(Type=Bone,Name="base_WatchChain_01"),Transform=(Rotation=(X=0,Y=0,Z=0.2588,W=0.9659),Translation=(X=0,Y=0.5,Z=-3),Scale3D=(X=1,Y=1,Z=1))))
    WeightMinimum=0
    Weight=@n7.Result
   )
   RigVMFunctionReturnNode(ExecuteContext=@RigVMFunction_Sequence.D)
   &RigUnit_ModifyTransforms(
    ExecutePin=()
    ItemToModify="((Item=(Type=Bone,Name=\"dyn_WatchChain_02_in\"),Transform=(Rotation=(X=0,Y=0,Z=0,W=1),Translation=(X=0,Y=1,Z=0),Scale3D=(X=1,Y=1,Z=1))),(Item=(Type=Bone,Name=\"base_WatchChain_01\"),Transform=(Rotation=(X=-0.0179,Y=0.017,Z=0.0265,W=0.9993),Translation=(X=0,Y= ...(truncated)"
    Weight=1
    WeightMinimum=0
   )
  ]}
 CollapseNode_ContainedGraph {nodes [
   n1 = RigVMFunctionEntryNode()
   RigVMFunction_Sequence = &RigVMFunction_Sequence(ExecuteContext=@n1.ExecuteContext)
   RigVMFunction_ControlFlowBranch = &RigVMFunction_ControlFlowBranch(ExecuteContext=@RigVMFunction_Sequence.A, Condition=RigVMVariableNode() {variable=hasVagabondFaceAcc}.Value)
   GetAlphaFromQuatDelta = RigVMFunctionReferenceNode(
    Rotation Order=ZYX
    child=(Type=Bone,Name="head")
    debugRotation=false
    outputMinMax=(X=0,Y=1)
    parent=(Type=Bone,Name="neck_01")
    premult=1
    rotationMinMax=(X=-20,Y=-40)
    triggerAxis=Z
    twistAxis=(X=1,Y=0,Z=0)
    ExecuteContext=@RigVMFunction_ControlFlowBranch.True
   )
   > &RigUnit_ModifyTransforms(
    ItemToModify="((Item=(Type=Bone,Name=\"collar_l_base\"),Transform=(Rotation=(X=0,Y=-0.5,Z=0,W=0.866),Translation=(X=-2,Y=0,Z=0),Scale3D=(X=1,Y=1,Z=1))),(Item=(Type=Bone,Name=\"collar_r_base\"),Transform=(Rotation=(X=0,Y=-0.5,Z=0,W=0.866),Translation=(X=-2,Y=0,Z=0),Scale3D=( ...(truncated)"
    WeightMinimum=0
    Weight=@GetAlphaFromQuatDelta.alpha
   )
   RigVMFunctionReturnNode(ExecuteContext=@RigVMFunction_Sequence.B)
  ]}
 Multiply_1_ContainedGraph {nodes [
   n1 = RigVMFunctionEntryNode()
   Multiply = &RigVMFunction_MathDoubleMul(A=@n1.A, B=@n1.B)
   RigVMFunctionReturnNode(Result=&RigVMFunction_MathDoubleMul(B=@n1.C, A=@Multiply.Result).Result)
  ]}
 Multiply_2_ContainedGraph {nodes [
   n1 = RigVMFunctionEntryNode()
   RigVMFunctionReturnNode(Result=&RigVMFunction_MathDoubleMul(B=@n1.C, A=&RigVMFunction_MathDoubleMul(A=@n1.A, B=@n1.B).Result).Result)
  ]}
 Multiply_2_ContainedGraph {nodes [
   n1 = RigVMFunctionEntryNode()
   RigVMFunctionReturnNode(Result=&RigVMFunction_MathDoubleMul(B=@n1.C, A=&RigVMFunction_MathDoubleMul(A=@n1.A, B=@n1.B).Result).Result)
  ]}
 Multiply_2_ContainedGraph {nodes [
   n1 = RigVMFunctionEntryNode()
   RigVMFunctionReturnNode(Result=&RigVMFunction_MathDoubleMul(B=@n1.C, A=&RigVMFunction_MathDoubleMul(A=@n1.A, B=@n1.B).Result).Result)
  ]}
 Multiply_6_ContainedGraph {nodes [
   n1 = RigVMFunctionEntryNode()
   RigVMFunctionReturnNode(Result=&RigVMFunction_MathDoubleMul(B=@n1.C, A=&RigVMFunction_MathDoubleMul(A=@n1.A, B=@n1.B).Result).Result)
  ]}
 Multiply_9_ContainedGraph {nodes [
   n1 = RigVMFunctionEntryNode()
   RigVMFunctionReturnNode(Result=&RigVMFunction_MathDoubleMul(B=@n1.C, A=&RigVMFunction_MathDoubleMul(A=@n1.A, B=@n1.B).Result).Result)
  ]}
 Multiply_9_ContainedGraph {nodes [
   n1 = RigVMFunctionEntryNode()
   RigVMFunctionReturnNode(Result=&RigVMFunction_MathDoubleMul(B=@n1.C, A=&RigVMFunction_MathDoubleMul(A=@n1.A, B=@n1.B).Result).Result)
  ]}
 New Function_ContainedGraph {nodes [
   RigVMFunctionEntryNode()
   RigVMFunction_MathBoolToInteger = &RigVMFunction_MathBoolToInteger(Value=RigVMVariableNode() {variable=onlyTwist}.Value)
   RigVMFunction_MathQuaternionSwingTwist = &RigVMFunction_MathQuaternionSwingTwist(
    Input=&RigUnit_GetRelativeTransformForItem(
     bChildInitial=False
     bParentInitial=False
     Child=RigVMVariableNode() {variable=child}.Value
     Parent=RigVMVariableNode() {variable=parent}.Value
    ).RelativeTransform.Rotation
    TwistAxis=RigVMVariableNode() {variable=twistAxis}.Value
   )
   n1 = RigVMDispatchNode(
    Values=((),())
    Index=@RigVMFunction_MathBoolToInteger.Result
    Values.0=@RigVMFunction_MathQuaternionSwingTwist.Swing
    Values.1=@RigVMFunction_MathQuaternionSwingTwist.Twist
   ) {template="DISPATCH_RigVMDispatch_SelectInt32(in Index,in Values,out Result)"}
   > RigVMVariableNode(Value=@n1.Result) {variable=entry_outputQuat}
   RigVMFunction_MathQuaternionToEuler = &RigVMFunction_MathQuaternionToEuler(Value=@n1.Result, RotationOrder=RigVMVariableNode() {variable=Rotation Order}.Value)
   > DISPATCH_RigVMDispatch_Print = RigVMDispatchNode(
    ScreenColor=(R=1,G=1,B=1,A=1)
    ScreenDuration=1.5
    Value=@RigVMFunction_MathQuaternionToEuler.Result
    Enabled=RigVMVariableNode() {variable=debugRotation}.Value
   ) {template="DISPATCH_RigVMDispatch_Print(in Prefix,in Value,in Enabled,in ScreenDuration,in ScreenColor)"}
   Cast = RigVMDispatchNode(Value=RigVMVariableNode() {variable=triggerAxis}.Value) {template="DISPATCH_RigVMDispatch_CastEnumToInt(in Value,out Result)"}
   DISPATCH_RigVMDispatch_SelectInt32 = RigVMDispatchNode(
    Values=(0,0,0)
    Index=@Cast.Result
    Values.0=@RigVMFunction_MathQuaternionToEuler.Result.X
    Values.1=@RigVMFunction_MathQuaternionToEuler.Result.Y
    Values.2=@RigVMFunction_MathQuaternionToEuler.Result.Z
   ) {template="DISPATCH_RigVMDispatch_SelectInt32(in Index,in Values,out Result)"}
   n2 = RigVMVariableNode() {variable=rotationMinMax}
   n3 = RigVMVariableNode() {variable=outputMinMax}
   Remap = &RigVMFunction_MathDoubleRemap(
    bClamp=False
    Value=@DISPATCH_RigVMDispatch_SelectInt32.Result
    SourceMinimum=@n2.Value.X
    SourceMaximum=@n2.Value.Y
    TargetMinimum=@n3.Value.X
    TargetMaximum=@n3.Value.Y
   )
   Absolute = &RigVMFunction_MathDoubleAbs(Value=@Remap.Result)
   n4 = RigVMVariableNode() {variable=absoluteValue}
   Multiply = &RigVMFunction_MathDoubleMul(
    A=&RigVMFunction_MathDoubleClamp(Value=RigVMDispatchNode(True=@Absolute.Result, False=@Remap.Result, Condition=@n4.Value) {template="DISPATCH_RigVMDispatch_If(in Condition,in True,in False,out Result)"}.Result).Result
    B=RigVMVariableNode() {variable=premult}.Value
   )
   > RigVMFunctionReturnNode(alpha=@Multiply.Result, rawQuat=RigVMVariableNode() {variable=entry_outputQuat}.Value)
  ]}
 RigVMFunctionLibrary {nodes [GetAlphaFromQuatDelta = RigVMCollapseNode(
    ExecuteContext=()
    Rotation Order=ZYX
    child=(Type=Bone,Name="None")
    outputMinMax=(X=0,Y=1)
    parent=(Type=Bone,Name="None")
    premult=1
    rotationMinMax=(X=0,Y=0)
    twistAxis=(X=1,Y=0,Z=0)
   )]}
 RigVMFunction_Sequence_1_ContainedGraph {nodes [
   n1 = RigVMFunctionEntryNode()
   RigVMFunction_Sequence = &RigVMFunction_Sequence(ExecuteContext=@n1.ExecuteContext)
   n2 = &RigVMFunction_Sequence(ExecuteContext=@RigVMFunction_Sequence.B)
   RigVMFunctionReturnNode(A=@RigVMFunction_Sequence.A, B=@n2.A, C=@n2.B)
  ]}
 RigVMFunction_Sequence_1_ContainedGraph {nodes [
   n1 = RigVMFunctionEntryNode()
   RigVMFunction_Sequence = &RigVMFunction_Sequence(ExecuteContext=@n1.ExecuteContext)
   n2 = &RigVMFunction_Sequence(ExecuteContext=@RigVMFunction_Sequence.B)
   n3 = &RigVMFunction_Sequence(ExecuteContext=@n2.B)
   RigVMFunctionReturnNode(A=@RigVMFunction_Sequence.A, B=@n2.A, C=@n3.A, D=@n3.B)
  ]}
 RigVMFunction_Sequence_1_ContainedGraph {nodes [
   n1 = RigVMFunctionEntryNode()
   RigVMFunction_Sequence = &RigVMFunction_Sequence(ExecuteContext=@n1.ExecuteContext)
   n2 = &RigVMFunction_Sequence(ExecuteContext=@RigVMFunction_Sequence.B)
   n3 = &RigVMFunction_Sequence(ExecuteContext=@n2.B)
   RigVMFunctionReturnNode(A=@RigVMFunction_Sequence.A, B=@n2.A, C=@n3.A, D=@n3.B)
  ]}
 RigVMFunction_Sequence_1_ContainedGraph {nodes [
   n1 = RigVMFunctionEntryNode()
   RigVMFunction_Sequence = &RigVMFunction_Sequence(ExecuteContext=@n1.ExecuteContext)
   n2 = &RigVMFunction_Sequence(ExecuteContext=@RigVMFunction_Sequence.B)
   RigVMFunctionReturnNode(A=@RigVMFunction_Sequence.A, B=@n2.A, C=@n2.B)
  ]}
 RigVMModel {nodes [
   RigUnit_BeginExecution = &RigUnit_BeginExecution()
   FN_DeformRig_FNC = RigVMFunctionReferenceNode(
    DeformRig_ Curve="(deformRig_l_upperarm_up_alpha_72_#1=False,deformRig_r_upperarm_up_alpha_73_#2=False,deformRig_l_upperarm_fr_alpha_74_#3=False,deformRig_r_upperarm_fr_alpha_75_CB51B8 ...(truncated)"
    DeformRig_Bool="(upperarm_1_#4=true,elbow_3_#5=true,wrist_7_#6=true,groin_8_#7=True,glute_17_#8=True,knee_9_40C408324CA261 ...(truncated)"
    DeformRig_Debug=(upperarm_debug_18_#4=False,wrist_debug_20_#6=False,lat_debug_26_#9=False,groinGlute_debug_27_#7=False)
    DeformRig_Value="(upperarm_up_18_#10=(Rotation=(X=0.0,Y=-0.0,Z=0.0,W=1.0),Translation=(X=0.0,Y=0.0,Z=8.0),Scale3D=(X=0.70,Y=1.0,Z=1.0)),upperarm_fwd_fr_19_#11=(Rotation=(X=0.0,Y=-0.0,Z=0.0,W=1.0),Translation=(X=0.0, ...(truncated)"
    ExecuteContext=@RigUnit_BeginExecution.ExecutePin
   )
   > RigVMFunction_Sequence = RigVMAggregateNode()
   DefiningVariables = RigVMCollapseNode(ExecuteContext=@RigVMFunction_Sequence.A)
   SurfaceConstraints = RigVMCollapseNode(ExecutePin=@DefiningVariables.ExecuteContext)
   QuadsPullJointSetup = RigVMCollapseNode(ExecuteContext=@RigVMFunction_Sequence.B)
   > FrontFlapsAntiColisionSetup = RigVMCollapseNode()
   > ChestChainsOffset = RigVMCollapseNode()
   VagabondFaceAccFixes = RigVMCollapseNode(ExecuteContext=@RigVMFunction_Sequence.C)
   ChestChains_RB = RigVMCollapseNode(ExecuteContext=())
   setFKCtrl = RigVMFunctionReferenceNode(ExecuteContext=(), setFK_Ctrl=false)
  ]}
 RigVMModel Build Graph {nodes [
   &RigUnit_PrepareForExecution()
   RigVMFunction_ControlFlowBranch = &RigVMFunction_ControlFlowBranch(Condition=False, ExecuteContext=())
   addFKCtrl = RigVMFunctionReferenceNode(
    Custom_FKCtrl=()
    ExecuteContext=()
    Options_FKCtrl="(pelvis_1_#12=true,spine_3_#13=true,clavicle_5_#14=true,arm_7_#15=true,leg_9_#16=true,head_11_46DC15EF465CD2979 ...(truncated)"
    Shape_Settings="(pelvis_2_#17=(bVisible=True,Name=\"Hexagon_Thin\",Color=(R=0.0117,G=0.8281,B=1,A=1),Transform=(Rotation=(X=0,Y=-0.7071,Z=0,W=0.7071),Translation=(X=2,Y=0,Z=0),Scale3D=(X=6,Y=5,Z=7))),spine_4_#18=(bVi ...(truncated)"
   )
  ]}
]
hierarchy [
 root {loc=<0 0 0>, rot=<0 0 0 1>}
 pelvis {parent=root, loc=<0 -0.7071 93.1191>, rot=<0.0127 0.707 -0.0127 -0.707>}
 spine_01 {parent=pelvis, loc=<6.7494 -0.6388 0>, rot=<0 0 0.034 -0.9994>}
 spine_02 {parent=spine_01, loc=<7.2518 -0.035 0>, rot=<0 0 0.0238 -0.9997>}
 spine_03 {parent=spine_02, loc=<6.0104 0 0>, rot=<0 0 0.0149 -0.9999>}
 spine_04 {parent=spine_03, loc=<8.1024 0 0>, rot=<0 0 -0.0132 -0.9999>}
 spine_05 {parent=spine_04, loc=<10.0146 0 0>, rot=<0 0 -0.0015 -1>}
 neck_01 {parent=spine_05, loc=<11.3138 0.1529 0>, rot=<0 0 -0.2003 -0.9797>}
 neck_02 {parent=neck_01, loc=<4.5627 0 0>, rot=<0 0 0.0504 -0.9987>}
 head {parent=neck_02, loc=<5.3914 0 0>, rot=<0 0 0.115 -0.9934>}
 clavicle_l {parent=spine_05, loc=<4.3575 -0.1466 -2.3223>, rot=<0.0465 -0.7289 -0.009 -0.6829>}
 upperarm_l {parent=clavicle_l, loc=<15.5111 0 0>, rot=<0.0226 -0.3234 -0.0088 -0.9459>}
 lowerarm_l {parent=upperarm_l, loc=<26.6514 0 0>, rot=<0.0391 -0.0099 -0.2453 -0.9686>}
 hand_l {parent=lowerarm_l, loc=<22.3491 0 0>, rot=<-0.7081 -0.0507 0.0975 -0.6975>}
 thumb_01_l {parent=hand_l, loc=<1.3118 -0.3191 -2.459>, rot=<0.6546 -0.4042 0.0664 -0.6354>}
 thumb_02_l {parent=thumb_01_l, loc=<5.9336 -0.0448 0.0401>, rot=<-0.0037 0.0213 0.298 -0.9543>}
 thumb_03_l {parent=thumb_02_l, loc=<2.2223 0.3682 -0.0907>, rot=<0.0118 -0.013 0.0066 -0.9998>}
 index_metacarpal_l {parent=hand_l, loc=<1.6535 0.0183 -2.96>, rot=<0.0158 -0.0124 -0.0173 -0.9996>}
 index_01_l {parent=index_metacarpal_l, loc=<8.3008 0.0357 -0.0524>, rot=<-0.0044 0.0485 0.0924 -0.9945>}
 index_02_l {parent=index_01_l, loc=<3.3587 -0.3004 0.0452>, rot=<-0.002 0.0027 0.3225 -0.9466>}
 index_03_l {parent=index_02_l, loc=<2.5578 -0.123 -0.0252>, rot=<-0.002 0.0067 0.2558 -0.9667>}
 middle_metacarpal_l {parent=hand_l, loc=<2.1856 -0.1725 -0.5984>, rot=<0.006 0.0168 -0.0329 -0.9993>}
 middle_01_l {parent=middle_metacarpal_l, loc=<7.3275 -0.2447 0.0129>, rot=<-0.0208 0.0042 0.0858 -0.9961>}
 middle_02_l {parent=middle_01_l, loc=<4.0123 0.1401 0.1175>, rot=<-0.0218 0.0258 0.3512 -0.9357>}
 middle_03_l {parent=middle_02_l, loc=<2.6077 0.0534 0.1373>, rot=<0.0082 0.0188 0.1708 -0.9851>}
 ring_metacarpal_l {parent=hand_l, loc=<2.2074 -0.3625 1.3638>, rot=<-0.0284 0.0722 -0.0458 -0.9959>}
 ring_01_l {parent=ring_metacarpal_l, loc=<7.0295 -0.6197 -0.1446>, rot=<-0.0489 -0.0252 0.1892 -0.9804>}
 ring_02_l {parent=ring_01_l, loc=<3.6174 0.4647 -0.1027>, rot=<0 0 0.2806 -0.9598>}
 ring_03_l {parent=ring_02_l, loc=<2.4152 0.114 0.0199>, rot=<0 0 0.1189 -0.9929>}
 pinky_metacarpal_l {parent=hand_l, loc=<1.7616 -0.9753 3.3814>, rot=<-0.0397 0.0814 -0.069 -0.9935>}
 pinky_01_l {parent=pinky_metacarpal_l, loc=<7.1709 -0.4831 -0.1094>, rot=<-0.0654 -0.0035 0.2242 -0.9723>}
 pinky_02_l {parent=pinky_01_l, loc=<2.8274 0.0893 -0.1043>, rot=<0.0037 0.0011 0.2854 -0.9584>}
 pinky_03_l {parent=pinky_02_l, loc=<2.2046 0.0307 -0.0078>, rot=<-0.0002 0.0098 0.1445 -0.9895>}
 deform_wrist_dn_base_l {parent=hand_l, loc=<0 -2.2342 0>, rot=<0 0 0 -1>}
 deform_wrist_dn_l {parent=deform_wrist_dn_base_l, loc=<0 0 0>, rot=<0 0 0 1>}
 deform_wrist_up_base_l {parent=hand_l, loc=<0 2.2682 0>, rot=<0 0 0 -1>}
 deform_wrist_up_l {parent=deform_wrist_up_base_l, loc=<0 0 0>, rot=<0 0 0 1>}
 lowerarm_twist_02_l {parent=lowerarm_l, loc=<9.6199 0 0>, rot=<0 0 0 -1>}
 lowerarm_twist_01_l {parent=lowerarm_l, loc=<17.8549 0 0>, rot=<0 0 0 -1>}
 deform_elbow_in_base_l {parent=lowerarm_l, loc=<0.2112 5.259 0.8274>, rot=<0 0 0 1>}
 deform_elbow_in_l {parent=deform_elbow_in_base_l, loc=<0 0 0>, rot=<0 0 0 1>}
 deform_elbow_out_base_l {parent=lowerarm_l, loc=<0 -4 0>, rot=<0 0 0 1>}
 deform_elbow_out_l {parent=deform_elbow_out_base_l, loc=<0 0 0>, rot=<0 0 0 1>}
 upperarm_twist_01_l {parent=upperarm_l, loc=<0.0041 0 0>, rot=<0 0 0 -1>}
 deform_upperarm_bk_base_l {parent=upperarm_twist_01_l, loc=<1.0813 -7.2193 -1.9453>, rot=<0 0 0 -1>}
 deform_upperarm_bk_l {parent=deform_upperarm_bk_base_l, loc=<0 0 0>, rot=<0 0 0 1>}
 deform_upperarm_fr_base_l {parent=upperarm_twist_01_l, loc=<-0.3691 7.0506 -2.765>, rot=<0 0 0 -1>}
 deform_upperarm_fr_l {parent=deform_upperarm_fr_base_l, loc=<0 0 0>, rot=<0 0 0 1>}
 deform_upperarm_up_base_l {parent=upperarm_twist_01_l, loc=<-3.2216 -1.1391 3.3311>, rot=<0 0 0 1>}
 deform_upperarm_up_l {parent=deform_upperarm_up_base_l, loc=<0 0 0>, rot=<0 0 0 1>}
 upperarm_twist_02_l {parent=upperarm_l, loc=<13.7941 0 0>, rot=<0 0 0 -1>}
 upperarm_l_null {parent=clavicle_l, type=Null, loc=<15.5111 0 0>, rot=<0.0226 -0.3234 -0.0088 -0.9459>}
 base_jacketFlap_top_01_l {parent=clavicle_l, loc=<6.0009 6.1249 4.1728>, rot=<0.8713 -0.1211 0.1022 0.4644>}
 dyn_jacketFlap_top_01_l {parent=base_jacketFlap_top_01_l, loc=<0 0 0>, rot=<0 0 0 1>}
 dyn_jacketFlap_top_02_l {parent=dyn_jacketFlap_top_01_l, loc=<5.8175 0 0>, rot=<0 -0.0635 0 -0.998>}
 clavicle_r {parent=spine_05, loc=<4.3575 -0.1466 2.3223>, rot=<0.7289 0.0465 -0.6829 0.009>}
 upperarm_r {parent=clavicle_r, loc=<-15.5111 0 0>, rot=<0.0226 -0.3234 -0.0088 -0.9459>}
 lowerarm_r {parent=upperarm_r, loc=<-26.6514 0 0>, rot=<0.0391 -0.0099 -0.2453 -0.9686>}
 hand_r {parent=lowerarm_r, loc=<-22.3491 0 0>, rot=<-0.7081 -0.0507 0.0975 -0.6975>}
 thumb_01_r {parent=hand_r, loc=<-1.3118 0.3191 2.459>, rot=<0.6546 -0.4042 0.0664 -0.6354>}
 thumb_02_r {parent=thumb_01_r, loc=<-5.9336 0.0448 -0.0401>, rot=<-0.0037 0.0213 0.298 -0.9543>}
 thumb_03_r {parent=thumb_02_r, loc=<-2.2223 -0.3682 0.0907>, rot=<0.0118 -0.013 0.0066 -0.9998>}
 index_metacarpal_r {parent=hand_r, loc=<-1.6535 -0.0183 2.96>, rot=<0.0158 -0.0124 -0.0173 -0.9996>}
 index_01_r {parent=index_metacarpal_r, loc=<-8.3008 -0.0357 0.0524>, rot=<-0.0044 0.0485 0.0924 -0.9945>}
 index_02_r {parent=index_01_r, loc=<-3.3587 0.3004 -0.0452>, rot=<-0.002 0.0027 0.3225 -0.9466>}
 index_03_r {parent=index_02_r, loc=<-2.5578 0.123 0.0252>, rot=<-0.002 0.0067 0.2558 -0.9667>}
 middle_metacarpal_r {parent=hand_r, loc=<-2.1856 0.1725 0.5984>, rot=<0.006 0.0168 -0.0329 -0.9993>}
 middle_01_r {parent=middle_metacarpal_r, loc=<-7.3275 0.2447 -0.0129>, rot=<-0.0208 0.0042 0.0858 -0.9961>}
 middle_02_r {parent=middle_01_r, loc=<-4.0123 -0.1401 -0.1175>, rot=<-0.0218 0.0258 0.3512 -0.9357>}
 middle_03_r {parent=middle_02_r, loc=<-2.6077 -0.0534 -0.1373>, rot=<0.0082 0.0188 0.1708 -0.9851>}
 ring_metacarpal_r {parent=hand_r, loc=<-2.2074 0.3625 -1.3638>, rot=<-0.0284 0.0722 -0.0458 -0.9959>}
 ring_01_r {parent=ring_metacarpal_r, loc=<-7.0295 0.6197 0.1446>, rot=<-0.0489 -0.0252 0.1892 -0.9804>}
 ring_02_r {parent=ring_01_r, loc=<-3.6174 -0.4647 0.1027>, rot=<0 0 0.2806 -0.9598>}
 ring_03_r {parent=ring_02_r, loc=<-2.4152 -0.114 -0.0199>, rot=<0 0 0.1189 -0.9929>}
 pinky_metacarpal_r {parent=hand_r, loc=<-1.7616 0.9753 -3.3814>, rot=<-0.0397 0.0814 -0.069 -0.9935>}
 pinky_01_r {parent=pinky_metacarpal_r, loc=<-7.1709 0.4831 0.1094>, rot=<-0.0654 -0.0035 0.2242 -0.9723>}
 pinky_02_r {parent=pinky_01_r, loc=<-2.8274 -0.0893 0.1043>, rot=<0.0037 0.0011 0.2854 -0.9584>}
 pinky_03_r {parent=pinky_02_r, loc=<-2.2046 -0.0307 0.0078>, rot=<-0.0002 0.0098 0.1445 -0.9895>}
 deform_wrist_dn_base_r {parent=hand_r, loc=<-0.0004 2.2345 -0.0001>, rot=<0 0 0 -1>}
 deform_wrist_dn_r {parent=deform_wrist_dn_base_r, loc=<0 0 0>, rot=<0 0 0 1>}
 deform_wrist_up_base_r {parent=hand_r, loc=<0.0003 -2.2684 0.0001>, rot=<0 0 0 -1>}
 deform_wrist_up_r {parent=deform_wrist_up_base_r, loc=<0 0 0>, rot=<0 0 0 1>}
 lowerarm_twist_02_r {parent=lowerarm_r, loc=<-9.6199 0 0>, rot=<0 0 0 -1>}
 lowerarm_twist_01_r {parent=lowerarm_r, loc=<-17.8549 0 0>, rot=<0 0 0 -1>}
 deform_elbow_in_base_r {parent=lowerarm_r, loc=<-0.2111 -5.2591 -0.8275>, rot=<0 0 0 -1>}
 deform_elbow_in_r {parent=deform_elbow_in_base_r, loc=<-0.0001 0 0.0002>, rot=<0 0 0 1>}
 deform_elbow_out_base_r {parent=lowerarm_r, loc=<0 4 -0.0001>, rot=<0 0 0 -1>}
 deform_elbow_out_r {parent=deform_elbow_out_base_r, loc=<0 0 0>, rot=<0 0 0 1>}
 upperarm_twist_01_r {parent=upperarm_r, loc=<-0.0041 0 0>, rot=<0 0 0 -1>}
 deform_upperarm_bk_base_r {parent=upperarm_twist_01_r, loc=<-1.0814 7.2192 1.9455>, rot=<0 0 0 -1>}
 deform_upperarm_bk_r {parent=deform_upperarm_bk_base_r, loc=<0 0 0>, rot=<0 0 0 1>}
 deform_upperarm_fr_base_r {parent=upperarm_twist_01_r, loc=<0.3689 -7.0506 2.7652>, rot=<0 0 0 -1>}
 deform_upperarm_fr_r {parent=deform_upperarm_fr_base_r, loc=<0 0 0>, rot=<0 0 0 1>}
 deform_upperarm_up_base_r {parent=upperarm_twist_01_r, loc=<3.2214 1.1391 -3.3308>, rot=<0 0 0 1>}
 deform_upperarm_up_r {parent=deform_upperarm_up_base_r, loc=<0 0 0>, rot=<0 0 0 -1>}
 upperarm_twist_02_r {parent=upperarm_r, loc=<-13.7941 0 0>, rot=<0 0 0 -1>}
 upperarm_r_null {parent=clavicle_r, type=Null, loc=<-15.5111 0 0>, rot=<0.0226 -0.3234 -0.0088 -0.9459>}
 base_jacketFlap_top_00_r {parent=clavicle_r, loc=<-1.6813 -7.7067 -3.4899>, rot=<0.0115 0.4434 0.8763 0.1883>}
 dyn_jacketFlap_top_00_r {parent=base_jacketFlap_top_00_r, loc=<0 0 0>, rot=<0 0 0 1>}
 base_jacketFlap_top_01_r {parent=clavicle_r, loc=<-6.0009 -6.1249 -4.1733>, rot=<-0.1022 0.4644 0.8713 0.1211>}
 dyn_jacketFlap_top_01_r {parent=base_jacketFlap_top_01_r, loc=<0 0 0>, rot=<0 0 0 1>}
 dyn_jacketFlap_top_02_r {parent=dyn_jacketFlap_top_01_r, loc=<5.8174 0.0005 0.0003>, rot=<0 -0.0635 0 -0.998>}
 clavicle_l_null {parent=spine_05, type=Null, loc=<4.3575 -0.1466 -2.3223>, rot=<0.0465 -0.7289 -0.009 -0.6829>}
 clavicle_r_null {parent=spine_05, type=Null, loc=<4.3575 -0.1466 2.3223>, rot=<0.7289 0.0465 -0.6829 0.009>}
 attach_cape {parent=spine_05, loc=<7.8468 -9.8595 0>, rot=<0.0616 -0.7044 -0.0616 -0.7044>}
 cape {parent=attach_cape, type=Null, loc=<0 0 0>, rot=<0 0 0 1>}
 attach_backpack {parent=spine_05, loc=<0.2533 -11.6211 0>, rot=<-0.0093 -0.707 0.0093 -0.707>}
 Backpack_BR {parent=attach_backpack, type=Null, loc=<0 1.5 1.12>, rot=<-0.0436 0 0 0.999>}
 base_bow {parent=spine_05, loc=<5.8381 9.2544 -0.2086>, rot=<0 0 -0.6055 -0.7959>}
 dyn_bow {parent=base_bow, loc=<0 0 0>, rot=<0 0 0 1>}
 collar_l_base {parent=spine_05, loc=<10.369 6.2061 -3.4785>, rot=<-0.0348 -0.8576 -0.385 -0.3394>}
 collar_l {parent=collar_l_base, loc=<0 0 0>, rot=<0 0 0 1>}
 collar_r_base {parent=spine_05, loc=<10.369 6.2061 3.4785>, rot=<0.3394 0.385 0.8576 0.0348>}
 collar_r {parent=collar_r_base, loc=<0 0 0>, rot=<0 0 0 1>}
 spine_05_null {parent=spine_04, type=Null, loc=<10.0146 0 0>, rot=<0 0 -0.0015 -1>}
 deform_pec_aim_base_l {parent=spine_04, loc=<7.8732 12.6388 -2.4123>, rot=<0.0102 -0.4602 0.3979 -0.7936>}
 deform_pec_aim_l {parent=deform_pec_aim_base_l, loc=<0 0 0>, rot=<0 0 0 1>}
 deform_pec_base_l {parent=deform_pec_aim_l, loc=<10.766 0 0>, rot=<0 0 0 1>}
 deform_pec_l {parent=deform_pec_base_l, loc=<0.055 0 0>, rot=<0 0 0 1>}
 deform_pec_aim_base_r {parent=spine_04, loc=<7.8727 12.6388 2.4123>, rot=<-0.4602 -0.0102 0.7936 0.3979>}
 deform_pec_aim_r {parent=deform_pec_aim_base_r, loc=<0 0 0>, rot=<0 0 0 1>}
 deform_pec_base_r {parent=deform_pec_aim_r, loc=<-10.7662 -0.0006 -0.0007>, rot=<0 0 0 1>}
 deform_pec_r {parent=deform_pec_base_r, loc=<0 0 0>, rot=<0 0 0 1>}
 base_jacketFlap_low_01_l {parent=spine_04, loc=<5.8395 11.6286 -3.941>, rot=<-0.6212 -0.3944 0.4522 -0.504>}
 dyn_jacketFlap_low_01_l {parent=base_jacketFlap_low_01_l, loc=<0 0 0>, rot=<0 0 0 1>}
 dyn_jacketFlap_low_02_l {parent=dyn_jacketFlap_low_01_l, loc=<4.1445 0 0>, rot=<0.002 0.1438 0.007 -0.9896>}
 base_jacketFlap_row_01_r {parent=spine_04, loc=<5.8397 11.6286 3.941>, rot=<-0.504 0.4522 -0.3944 -0.6212>}
 dyn_jacketFlap_row_01_r {parent=base_jacketFlap_row_01_r, loc=<0 0 0>, rot=<0 0 0 1>}
 dyn_jacketFlap_row_02_r {parent=dyn_jacketFlap_row_01_r, loc=<4.1446 -0.0002 -0.0001>, rot=<-0.002 0.1438 -0.007 -0.9896>}
 spine_04_null {parent=spine_03, type=Null, loc=<8.1024 0 0>, rot=<0 0 -0.0132 -0.9999>}
 base_WatchChain_01 {parent=spine_03, loc=<4.7312 11.2729 -8.1987>, rot=<-0.2234 -0.9403 0.2494 0.0601>}
 dyn_WatchChain_01 {parent=base_WatchChain_01, loc=<0 0 0>, rot=<0 0 0 1>}
 dyn_WatchChain_02 {parent=dyn_WatchChain_01, loc=<4.5003 -0.0214 0.0012>, rot=<0.0014 0.1004 -0.0085 -0.9949>}
 dyn_WatchChain_03 {parent=dyn_WatchChain_02, loc=<5.823 0 0>, rot=<0.0004 0.0385 -0.0098 -0.9992>}
 base_WatchChain_02_in {parent=dyn_WatchChain_02, loc=<0 0 -6.2054>, rot=<0 0 0 1>}
 dyn_WatchChain_02_in {parent=base_WatchChain_02_in, loc=<0 0 0>, rot=<0 0 0 1>}
 base_WatchChain_01_in {parent=dyn_WatchChain_01, loc=<0.9919 -0.6177 -7.2631>, rot=<0.1429 0.0343 -0.005 -0.9891>}
 dyn_WatchChain_01_in {parent=base_WatchChain_01_in, loc=<0 0 0>, rot=<0 0 0 1>}
 spine_03_null {parent=spine_02, type=Null, loc=<6.0104 0 0>, rot=<0 0 0.0149 -0.9999>}
 base_SuitBackFlap_01_l {parent=spine_02, loc=<3.2716 -12.8437 -7.1064>, rot=<0.1982 -0.9683 -0.1517 -0.0127>}
 dyn_SuitBackFlap_01_l {parent=base_SuitBackFlap_01_l, loc=<0 0 0>, rot=<0 0 0 1>}
 dyn_SuitBackFlap_02_l {parent=dyn_SuitBackFlap_01_l, loc=<5.9104 0 0>, rot=<0 0 0 1>}
 base_SuitBackFlap_01_r {parent=spine_02, loc=<3.2713 -12.8438 7.1064>, rot=<0.1982 -0.9683 0.1517 0.0127>}
 dyn_SuitBackFlap_01_r {parent=base_SuitBackFlap_01_r, loc=<0 0 0>, rot=<0 0 0 1>}
 dyn_SuitBackFlap_02_r {parent=dyn_SuitBackFlap_01_r, loc=<5.9101 -0.0001 0>, rot=<0 0 0 1>}
 spine_02_RB_Helper_base {parent=spine_02, loc=<-0.7784 15.9702 -9.0134>, rot=<0.0282 -0.7065 -0.0282 -0.7065>}
 spine_02_RB_Helper {parent=spine_02_RB_Helper_base, loc=<0 0 0>, rot=<0 0 0 1>}
 bellyJacketFlap_RB_Helper_base {parent=spine_02, loc=<3.1718 17.8922 -2.5496>, rot=<-0.0513 -0.5898 0.2348 -0.771>}
 bellyJacketFlap_RB_Helper {parent=bellyJacketFlap_RB_Helper_base, loc=<0 0 0>, rot=<0 0 0 1>}
 spine_05_DeltaSpine1_null {parent=spine_01, type=Null, loc=<31.3352 -1.456 0>, rot=<0 0 -0.024 0.9997>}
 spine_02_null {parent=spine_01, type=Null, loc=<7.2518 -0.035 0>, rot=<0 0 0.0238 -0.9997>}
 spine_03_DeltaSpine01_null {parent=spine_01, type=Null, loc=<13.2554 -0.3205 0>, rot=<0 0 -0.0386 0.9993>}
 thigh_l_null {parent=pelvis, type=Null, loc=<-2.5693 -0.0109 -7.7501>, rot=<0.0012 0.0572 0.0155 -0.9982>}
 spine_01_null {parent=pelvis, type=Null, loc=<6.7494 -0.6388 0>, rot=<0 0 0.034 -0.9994>}
 thigh_l {parent=pelvis, loc=<-2.5693 -0.0109 -7.7501>, rot=<0.0012 0.0572 0.0155 -0.9982>}
 calf_l {parent=thigh_l, loc=<-39.0484 0 0>, rot=<0.0006 -0.0161 -0.0391 -0.9991>}
 foot_l {parent=calf_l, loc=<-42.561 0.1077 0.4292>, rot=<0.0418 -0.0455 0.0343 -0.9975>}
 ball_l {parent=foot_l, loc=<-6.1599 13.8264 0.6806>, rot=<0.0001 -0.0001 -0.7063 -0.7079>}
 calf_twist_02_l {parent=calf_l, loc=<-18.8064 -0.0967 0.0397>, rot=<0.0237 0.0095 -0.0035 -0.9997>}
 calf_twist_01_l {parent=calf_l, loc=<-35.813 -0.0967 0.0397>, rot=<0.0237 0.0095 -0.0035 -0.9997>}
 deform_knee_in_base_l {parent=calf_l, loc=<1.298 -5.2957 -0.0496>, rot=<0 0 0 -1>}
 deform_knee_in_l {parent=deform_knee_in_base_l, loc=<0 0 0>, rot=<0 0 0 1>}
 deform_knee_out_base_l {parent=calf_l, loc=<0 4 0>, rot=<0 0 0 -1>}
 deform_knee_out_l {parent=deform_knee_out_base_l, loc=<0 0 0>, rot=<0 0 0 1>}
 thigh_twist_01_l {parent=thigh_l, loc=<-21.6303 0 0>, rot=<0.0013 0 0 -1>}
 deform_groin_base_l {parent=thigh_l, loc=<-2.1908 8.1142 -1.7442>, rot=<0 0 0 -1>}
 deform_groin_l {parent=deform_groin_base_l, loc=<0 0 0>, rot=<0 0 0 1>}
 quadBack_low_l_base {parent=thigh_l, loc=<-30.9394 -11.0409 -3.703>, rot=<0.0013 0 0 -1>}
 quadBack_low_l {parent=quadBack_low_l_base, loc=<0 0 0>, rot=<0 0 0 1>}
 quadBack_hi_l_base {parent=thigh_l, loc=<-20.5106 -12.7702 -4.3443>, rot=<0.0013 0 0 -1>}
 quadBack_hi_l {parent=quadBack_hi_l_base, loc=<0 0 0>, rot=<0 0 0 1>}
 InnerLeg_RB_Helper_base {parent=thigh_l, loc=<-12.8982 15.0733 -4.3489>, rot=<-0.0019 -0.7464 0.0016 -0.6655>}
 InnerLeg_RB_Helper {parent=InnerLeg_RB_Helper_base, loc=<0 0 0>, rot=<0 0 0 1>}
 thigh_r {parent=pelvis, loc=<-2.5693 -0.0109 7.7501>, rot=<0.0572 -0.0012 0.9982 0.0155>}
 calf_r {parent=thigh_r, loc=<39.0484 0 0>, rot=<0.0006 -0.0161 -0.0391 -0.9991>}
 foot_r {parent=calf_r, loc=<42.561 -0.1077 -0.4292>, rot=<0.0418 -0.0455 0.0343 -0.9975>}
 ball_r {parent=foot_r, loc=<6.1599 -13.8264 -0.6806>, rot=<0.0001 -0.0001 -0.7063 -0.7079>}
 calf_twist_02_r {parent=calf_r, loc=<18.8064 0.0967 -0.0397>, rot=<0.0237 0.0095 -0.0035 -0.9997>}
 calf_twist_01_r {parent=calf_r, loc=<35.813 0.0967 -0.0397>, rot=<0.0237 0.0095 -0.0035 -0.9997>}
 deform_knee_in_base_r {parent=calf_r, loc=<-1.298 5.2957 0.0496>, rot=<0 0 0 1>}
 deform_knee_in_r {parent=deform_knee_in_base_r, loc=<0 0 0>, rot=<0 0 0 1>}
 deform_knee_out_base_r {parent=calf_r, loc=<0 -4 0>, rot=<0 0 0 -1>}
 deform_knee_out_r {parent=deform_knee_out_base_r, loc=<0 0 0>, rot=<0 0 0 1>}
 thigh_twist_01_r {parent=thigh_r, loc=<21.6303 0 0>, rot=<0.0013 0 0 -1>}
 deform_groin_base_r {parent=thigh_r, loc=<2.1908 -8.1142 1.7442>, rot=<0 0 0 -1>}
 deform_groin_r {parent=deform_groin_base_r, loc=<0 0 0>, rot=<0 0 0 1>}
 quadBack_low_r_base {parent=thigh_r, loc=<30.9394 11.0409 3.703>, rot=<0.0013 0 0 -1>}
 quadBack_low_r {parent=quadBack_low_r_base, loc=<0 0 0>, rot=<0 0 0 1>}
 quadBack_hi_r_base {parent=thigh_r, loc=<20.5106 12.7702 4.3443>, rot=<0.0013 0 0 -1>}
 quadBack_hi_r {parent=quadBack_hi_r_base, loc=<0 0 0>, rot=<0 0 0 1>}
 deform_glute_base_l {parent=pelvis, loc=<-8.8064 -8.8222 -8.4287>, rot=<0.0275 0.0501 -0.4472 -0.8926>}
 deform_glute_l {parent=deform_glute_base_l, loc=<0 0 0>, rot=<0 0 0 1>}
 deform_glute_base_r {parent=pelvis, loc=<-8.8062 -8.8222 8.4287>, rot=<0.0501 -0.0275 0.8926 -0.4472>}
 deform_glute_r {parent=deform_glute_base_r, loc=<0 0 0>, rot=<0 0 0 1>}
 pelvis_RB_Helper_base {parent=pelvis, loc=<3.0757 17.8208 -10.5431>, rot=<-0.0127 -0.707 0.0127 -0.707>}
 pelvis_RB_Helper {parent=pelvis_RB_Helper_base, loc=<0 0 0>, rot=<0 0 0 1>}
 pelvisBellyPush_l_base {parent=pelvis, loc=<0.0028 11.5626 -4.8195>, rot=<-0.363 -0.5838 -0.4465 -0.5728>}
 pelvisBellyPush_l {parent=pelvisBellyPush_l_base, loc=<0 0 0>, rot=<0 0 0 1>}
 pelvisBellyPush_r_base {parent=pelvis, loc=<0.0029 11.5625 4.8195>, rot=<0.5728 0.4465 0.5838 0.363>}
 pelvisBellyPush_r {parent=pelvisBellyPush_r_base, loc=<0 0 0>, rot=<0 0 0 1>}
 calf_l_null {type=Null, loc=<12.2078 -0.9999 51.7592>, rot=<0.0315 0.7349 -0.027 -0.6769>}
 calf_r_null {type=Null, loc=<-12.2078 -0.9999 51.7592>, rot=<-0.6769 0.027 0.7349 -0.0315>}
]
variables [
 AnimDynamicsEnabled {type=bool, default=False}
 RigidBodiesEnabled {type=bool, default=False}
 Spine01_ForwardBend {type=double, default=0}
 fullSpineBend {type=double, default=0}
 hasVagabondFaceAcc {type=bool, default=False}
 isCrouchTurning {type=float, default=0}
 isCrouchWalking {type=double, default=0}
 isCrouchingStill {type=float, default=0}
 isCrouchingb {type=bool, default=False}
 isTargetting {type=bool, default=False}
 spine03To01BendBackwards {type=double, default=0}
 spineTwistingRight {type=double, default=0}
 standing_L_ChainLegRaise {type=double, default=0}
]
