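# Post-copy Anim Blueprint for the RoyalWonk medium male cosmetic body, adding secondary motion to clothing and accessories.
# The parent mesh pose runs through a Control Rig, RigidBody chain physics, then AnimDynamics for bow and jacket flaps.
# Cached poses feed a final bool blend between simulated suit back flaps and a fixed rotation offset.
# Each update computes state-based gravity override and flags for cape crouching, Vagabond face detection and rigid-body fixes.
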AnimBlueprint /BRCosmetics/Characters/Player/Male/Medium/Bodies/M_MED_RoyalWonk/Meshes/M_MED_RoyalWonk_AnimBP
target_skeleton=&/M_MED_RoyalWonk_Skeleton
graphs [
 AnimGraph {nodes [
   n1 = "Transform (Modify) Bone - Bone: dyn_WatchChain_01"!(ComponentPose, bAlphaBoolEnabled:bool=True, Translation="0,0,0", Rotation="0,-1,-1") -> (Pose) {
    alphaInputType=Bool
    rotationMode=BMM_Additive
    rotationSpace=BCS_BoneSpace
    translationMode=BMM_Additive
    translationSpace=BCS_BoneSpace
    alphaBoolBlend {blendInTime=0.2, blendOutTime=0.2}
    boneToModify {boneName=dyn_WatchChain_01}
    rotation {roll=-1, yaw=-1}
   }
   "Transform (Modify) Bone - Bone: dyn_WatchChain_02"!(ComponentPose=@n1, bAlphaBoolEnabled:bool=True, Translation="0,0,0", Rotation="0,0,-1") -> (Pose) {
    alphaInputType=Bool
    rotationMode=BMM_Additive
    rotationSpace=BCS_BoneSpace
    translationMode=BMM_Additive
    translationSpace=BCS_BoneSpace
    alphaBoolBlend {blendInTime=0.2, blendOutTime=0.2}
    boneToModify {boneName=dyn_WatchChain_02}
    rotation {roll=-1}
   }
   Output Pose(Result=Use cached pose 'DynamicsCache'()) {layerGroup=None}
   n2 = Control Rig!(
    Source=Copy Pose From Mesh() {bUseAttachedParent=true}
    Alpha
    isTargetting:bool=false
    isCrouchingb:bool=false
    AnimDynamicsEnabled:bool=false
    hasVagabondFaceAcc:bool=false
   ) -> (Pose) {controlRigAssetReference {blueprintRigClass=&/M_MED_RoyalWonk_CtrlRig.M_MED_RoyalWonk_CtrlRig_C}}
   Save cached pose 'CRCache_Input'(Pose=@n2) {cacheName=CRCache_Input}
   n3 = Anim Dynamics - Bone: dyn_SuitBackFlap_01_l!(ComponentPose=Local To Component(LocalPose=Use cached pose 'preFlapDynCache'()), Alpha=1, GravityOverride:Vector="0,0,-980") -> (Pose) {
    angularDampingOverride=0.999
    angularSpringConstant=10000
    bAngularSpring=true
    bOverrideAngularDamping=true
    bUseGravityOverride=true
    gravityScale=0.35
    simulationSpace=RootRelative
    alphaBoolBlend {blendOption=Linear}
    boundBone {boneName=dyn_SuitBackFlap_01_l}
    componentAppliedLinearAccClamp=<5000 5000 5000>
    componentLinearAccScale=<1 1 1>
    componentLinearVelScale=<1 1 1>
    gravityOverride {z=-980}
    physicsBodyDefinitions [{
      boundBone {boneName=dyn_SuitBackFlap_01_l}
      boxExtents=<24 42 42>
      constraintSetup {angularLimitsMax=<0 0 0>, angularLimitsMin=<0 -10 -45>, angularTarget=<1 0 0>}
      localJointOffset=<3 0 0>
     }]
    simSpaceSettings {maxAngularAcceleration=30, maxAngularVelocity=5, simSpaceAngularAlpha=1}
   }
   n4 = Anim Dynamics - Bone: dyn_SuitBackFlap_02_l!(ComponentPose=@n3, Alpha=0.9, GravityOverride:Vector="0,0,-980") -> (Pose) {
    angularDampingOverride=0.9999
    angularSpringConstant=10000
    bAngularSpring=true
    bOverrideAngularDamping=true
    bUseGravityOverride=true
    gravityScale=0.35
    simulationSpace=RootRelative
    alphaBoolBlend {blendOption=Linear}
    boundBone {boneName=dyn_SuitBackFlap_02_l}
    componentAppliedLinearAccClamp=<5000 5000 5000>
    componentLinearAccScale=<1 1 1>
    componentLinearVelScale=<1 1 1>
    gravityOverride {z=-980}
    physicsBodyDefinitions [{
      boundBone {boneName=dyn_SuitBackFlap_02_l}
      boxExtents=<22 40 40>
      constraintSetup {angularLimitsMax=<0 0 0>, angularLimitsMin=<0 -20 -30>, angularTarget=<1 0 0>}
      localJointOffset=<3 0 0>
     }]
    simSpaceSettings {maxAngularAcceleration=30, maxAngularVelocity=5, simSpaceAngularAlpha=1}
   }
   n5 = Anim Dynamics - Bone: dyn_SuitBackFlap_01_r!(ComponentPose=@n4, Alpha=0.8, GravityOverride:Vector="0,0,-980") -> (Pose) {
    angularDampingOverride=0.9999
    angularSpringConstant=10000
    bAngularSpring=true
    bOverrideAngularDamping=true
    bUseGravityOverride=true
    gravityScale=0.35
    simulationSpace=RootRelative
    alphaBoolBlend {blendOption=Linear}
    boundBone {boneName=dyn_SuitBackFlap_01_r}
    componentAppliedLinearAccClamp=<5000 5000 5000>
    componentLinearAccScale=<1 1 1>
    componentLinearVelScale=<1 1 1>
    gravityOverride {z=-980}
    physicsBodyDefinitions [{
      boundBone {boneName=dyn_SuitBackFlap_01_r}
      boxExtents=<22 40 40>
      constraintSetup {angularLimitsMax=<0 5 0>, angularLimitsMin=<0 -10 -30>, angularTarget=<1 0 0>}
      localJointOffset=<3 0 0>
     }]
    simSpaceSettings {maxAngularAcceleration=30, maxAngularVelocity=5, simSpaceAngularAlpha=1}
   }
   n6 = Anim Dynamics - Bone: dyn_SuitBackFlap_02_r!(ComponentPose=@n5, Alpha=0.9, GravityOverride:Vector="0,0,-980") -> (Pose) {
    angularDampingOverride=0.9999
    angularSpringConstant=10000
    bAngularSpring=true
    bOverrideAngularDamping=true
    bUseGravityOverride=true
    gravityScale=0.35
    simulationSpace=RootRelative
    alphaBoolBlend {blendOption=Linear}
    boundBone {boneName=dyn_SuitBackFlap_02_r}
    componentAppliedLinearAccClamp=<5000 5000 5000>
    componentLinearAccScale=<1 1 1>
    componentLinearVelScale=<1 1 1>
    gravityOverride {z=-980}
    physicsBodyDefinitions [{
      boundBone {boneName=dyn_SuitBackFlap_02_r}
      boxExtents=<22 40 40>
      constraintSetup {angularLimitsMax=<5 0 0>, angularLimitsMin=<-5 -20 -40>, angularTarget=<1 0 0>}
      localJointOffset=<2.5 0 0>
     }]
    simSpaceSettings {maxAngularAcceleration=30, maxAngularVelocity=5, simSpaceAngularAlpha=1}
   }
   n7 = "Transform (Modify) Bone - Bone: dyn_SuitBackFlap_02_l"!(ComponentPose=Local To Component(LocalPose=Use cached pose 'preFlapDynCache'()), Alpha=1, Rotation="0,7.5,0") -> (Pose) {rotationMode=BMM_Additive, rotationSpace=BCS_BoneSpace, boneToModify {boneName=dyn_SuitBackFlap_02_l}, rotation {yaw=7.5}}
   n8 = "Transform (Modify) Bone - Bone: dyn_SuitBackFlap_02_r"!(ComponentPose=@n7, Alpha=1, Rotation="0,7.5,0") -> (Pose) {rotationMode=BMM_Additive, rotationSpace=BCS_BoneSpace, boneToModify {boneName=dyn_SuitBackFlap_02_r}, rotation {yaw=7.5}}
   Save cached pose 'DynamicsCache'(Pose=Blend Poses by bool(BlendPose_0=Component To Local(ComponentPose=@n8), BlendPose_1=Component To Local(ComponentPose=@n6), BlendTime_0=0.2, BlendTime_1=0.2)) {cacheName=DynamicsCache}
   n9 = RigidBody!(ComponentPose=Local To Component(LocalPose=Use cached pose 'CRCache_Input'()), OverrideWorldGravity:Vector="0,0,0") -> (Pose) {
    bOverrideWorldGravity=true
    overridePhysicsAsset=&/M_MED_RoyalWonk_Chain_Physics
    simulationSpace=BaseBoneSpace
    alphaBoolBlend {blendOption=Linear}
    baseBoneRef {boneName=spine_05}
    componentAppliedLinearAccClamp=<3000 3000 3000>
    overrideWorldGravity {z=-980}
    simSpaceSettings {
     dampingAlpha=0
     maxAngularAcceleration=1
     maxAngularVelocity=1
     maxLinearAcceleration=1200
     maxLinearVelocity=500
     velocityScaleZ=0.6
     worldAlpha=1
     externalLinearDragV=<0.2 0.2 0.2>
    }
   }
   n10 = "Transform (Modify) Bone - Bone: dyn_WatchChain_01"(ComponentPose=@n9, bAlphaBoolEnabled=True, Translation="0,-0.5,0", Rotation="0,-1,-1") {
    alphaInputType=Bool
    rotationMode=BMM_Additive
    rotationSpace=BCS_BoneSpace
    translationMode=BMM_Additive
    translationSpace=BCS_BoneSpace
    alphaBoolBlend {blendInTime=0.2, blendOutTime=0.2}
    boneToModify {boneName=dyn_WatchChain_01}
    rotation {roll=-1, yaw=-1}
    translation {y=-0.5}
   }
   n11 = Anim Dynamics - Bone: dyn_bow!(ComponentPose=@n10, Alpha=1, GravityOverride:Vector="0,0,-980") -> (Pose) {
    angularDampingOverride=1
    angularSpringConstant=20000
    bAngularSpring=true
    bOverrideAngularDamping=true
    bUseGravityOverride=true
    gravityScale=0.35
    simulationSpace=RootRelative
    boundBone {boneName=dyn_bow}
    componentAppliedLinearAccClamp=<5000 5000 5000>
    componentLinearAccScale=<1 1 1>
    componentLinearVelScale=<1 1 1>
    gravityOverride {z=-980}
    physicsBodyDefinitions [{
      boundBone {boneName=dyn_bow}
      boxExtents=<31 42 42>
      constraintSetup {angularLimitsMax=<0 7.5 7.5>, angularLimitsMin=<0 -7.5 -7.5>, angularTarget=<1 0 0>}
      localJointOffset=<0.5 0 0>
     }]
    simSpaceSettings {maxAngularAcceleration=25, maxAngularVelocity=5, simSpaceAngularAlpha=1}
   }
   n12 = "Transform (Modify) Bone - Bone: dyn_bow"!(ComponentPose=@n11, bAlphaBoolEnabled:bool=True, Translation="0,0,0", Scale="0.6,1,1") -> (Pose) {
    alphaInputType=Bool
    scaleMode=BMM_Additive
    scaleSpace=BCS_BoneSpace
    translationMode=BMM_Additive
    translationSpace=BCS_BoneSpace
    boneToModify {boneName=dyn_bow}
    scale {x=0.6}
   }
   n13 = Anim Dynamics - Bone: dyn_jacketFlap_top_01_l!(ComponentPose=@n12, Alpha=1, GravityOverride:Vector="0,0,-980") -> (Pose) {
    angularDampingOverride=1
    angularSpringConstant=10000
    bAngularSpring=true
    bOverrideAngularDamping=true
    bUseGravityOverride=true
    gravityScale=0.35
    simulationSpace=RootRelative
    alphaBoolBlend {blendOption=Linear}
    boundBone {boneName=dyn_jacketFlap_top_01_l}
    componentAppliedLinearAccClamp=<5000 5000 5000>
    componentLinearAccScale=<1 1 1>
    componentLinearVelScale=<1 1 1>
    gravityOverride {z=-980}
    physicsBodyDefinitions [{
      boundBone {boneName=dyn_jacketFlap_top_01_l}
      boxExtents=<20 30 30>
      constraintSetup {angularLimitsMax=<0 10 5>, angularLimitsMin=<0 0 -5>, angularTarget=<1 0 0>}
      localJointOffset=<3 0 0>
     }]
    simSpaceSettings {maxAngularAcceleration=20, maxAngularVelocity=5, simSpaceAngularAlpha=1}
   }
   n14 = Anim Dynamics - Bone: dyn_jacketFlap_top_02_l!(ComponentPose=@n13, Alpha=1, GravityOverride:Vector="0,0,-980") -> (Pose) {
    angularDampingOverride=1
    angularSpringConstant=10000
    bAngularSpring=true
    bOverrideAngularDamping=true
    bUseGravityOverride=true
    gravityScale=0.35
    simulationSpace=RootRelative
    alphaBoolBlend {blendOption=Linear}
    boundBone {boneName=dyn_jacketFlap_top_02_l}
    componentAppliedLinearAccClamp=<5000 5000 5000>
    componentLinearAccScale=<1 1 1>
    componentLinearVelScale=<1 1 1>
    gravityOverride {z=-980}
    physicsBodyDefinitions [{
      boundBone {boneName=dyn_jacketFlap_top_02_l}
      boxExtents=<20 30 30>
      constraintSetup {angularLimitsMax=<0 10 5>, angularLimitsMin=<0 0 -5>, angularTarget=<1 0 0>}
      localJointOffset=<3 0 0>
     }]
    simSpaceSettings {maxAngularAcceleration=20, maxAngularVelocity=5, simSpaceAngularAlpha=1}
   }
   n15 = Anim Dynamics - Bone: dyn_jacketFlap_top_01_r!(ComponentPose=@n14, Alpha=1, GravityOverride:Vector="0,0,-980") -> (Pose) {
    angularDampingOverride=1
    angularSpringConstant=10000
    bAngularSpring=true
    bOverrideAngularDamping=true
    bUseGravityOverride=true
    gravityScale=0.35
    simulationSpace=RootRelative
    alphaBoolBlend {blendOption=Linear}
    boundBone {boneName=dyn_jacketFlap_top_01_r}
    componentAppliedLinearAccClamp=<5000 5000 5000>
    componentLinearAccScale=<1 1 1>
    componentLinearVelScale=<1 1 1>
    gravityOverride {z=-980}
    physicsBodyDefinitions [{
      boundBone {boneName=dyn_jacketFlap_top_01_r}
      boxExtents=<20 30 30>
      constraintSetup {angularLimitsMax=<0 10 5>, angularLimitsMin=<0 0 -5>, angularTarget=<1 0 0>}
      localJointOffset=<3 0 0>
     }]
    simSpaceSettings {maxAngularAcceleration=20, maxAngularVelocity=5, simSpaceAngularAlpha=1}
   }
   n16 = Anim Dynamics - Bone: dyn_jacketFlap_top_02_r!(ComponentPose=@n15, Alpha=1, GravityOverride:Vector="0,0,-980") -> (Pose) {
    angularDampingOverride=1
    angularSpringConstant=10000
    bAngularSpring=true
    bOverrideAngularDamping=true
    bUseGravityOverride=true
    gravityScale=0.35
    simulationSpace=RootRelative
    alphaBoolBlend {blendOption=Linear}
    boundBone {boneName=dyn_jacketFlap_top_02_r}
    componentAppliedLinearAccClamp=<5000 5000 5000>
    componentLinearAccScale=<1 1 1>
    componentLinearVelScale=<1 1 1>
    gravityOverride {z=-980}
    physicsBodyDefinitions [{
      boundBone {boneName=dyn_jacketFlap_top_02_r}
      boxExtents=<20 30 30>
      constraintSetup {angularLimitsMax=<0 10 5>, angularLimitsMin=<0 0 -5>, angularTarget=<1 0 0>}
      localJointOffset=<3 0 0>
     }]
    simSpaceSettings {maxAngularAcceleration=20, maxAngularVelocity=5, simSpaceAngularAlpha=1}
   }
   n17 = Anim Dynamics - Bone: dyn_jacketFlap_low_01_l!(ComponentPose=@n16, Alpha=1, GravityOverride:Vector="0,0,-980") -> (Pose) {
    angularDampingOverride=1
    angularSpringConstant=10000
    bAngularSpring=true
    bOverrideAngularDamping=true
    bUseGravityOverride=true
    gravityScale=0.35
    simulationSpace=RootRelative
    alphaBoolBlend {blendOption=Linear}
    boundBone {boneName=dyn_jacketFlap_low_01_l}
    componentAppliedLinearAccClamp=<5000 5000 5000>
    componentLinearAccScale=<1 1 1>
    componentLinearVelScale=<1 1 1>
    gravityOverride {z=-980}
    physicsBodyDefinitions [{
      boundBone {boneName=dyn_jacketFlap_low_01_l}
      boxExtents=<20 30 30>
      constraintSetup {angularLimitsMax=<0 5 5>, angularLimitsMin=<0 0 -5>, angularTarget=<1 0 0>}
      localJointOffset=<3 0 0>
     }]
    simSpaceSettings {maxAngularAcceleration=25, maxAngularVelocity=5, simSpaceAngularAlpha=1}
   }
   n18 = Anim Dynamics - Bone: dyn_jacketFlap_low_02_l!(ComponentPose=@n17, Alpha=1, GravityOverride:Vector="0,0,-980") -> (Pose) {
    angularDampingOverride=1
    angularSpringConstant=10000
    bAngularSpring=true
    bOverrideAngularDamping=true
    bUseGravityOverride=true
    gravityScale=0.35
    simulationSpace=RootRelative
    alphaBoolBlend {blendOption=Linear}
    boundBone {boneName=dyn_jacketFlap_low_02_l}
    componentAppliedLinearAccClamp=<5000 5000 5000>
    componentLinearAccScale=<1 1 1>
    componentLinearVelScale=<1 1 1>
    gravityOverride {z=-980}
    physicsBodyDefinitions [{
      boundBone {boneName=dyn_jacketFlap_low_02_l}
      boxExtents=<20 30 30>
      constraintSetup {angularLimitsMax=<0 5 10>, angularLimitsMin=<0 0 -10>, angularTarget=<1 0 0>}
      localJointOffset=<3 0 0>
     }]
    simSpaceSettings {maxAngularAcceleration=25, maxAngularVelocity=5, simSpaceAngularAlpha=1}
   }
   n19 = Anim Dynamics - Bone: dyn_jacketFlap_row_01_r!(ComponentPose=@n18, Alpha=1, GravityOverride:Vector="0,0,-980") -> (Pose) {
    angularDampingOverride=1
    angularSpringConstant=10000
    bAngularSpring=true
    bOverrideAngularDamping=true
    bUseGravityOverride=true
    gravityScale=0.35
    simulationSpace=RootRelative
    alphaBoolBlend {blendOption=Linear}
    boundBone {boneName=dyn_jacketFlap_row_01_r}
    componentAppliedLinearAccClamp=<5000 5000 5000>
    componentLinearAccScale=<1 1 1>
    componentLinearVelScale=<1 1 1>
    gravityOverride {z=-980}
    physicsBodyDefinitions [{
      boundBone {boneName=dyn_jacketFlap_row_01_r}
      boxExtents=<20 30 30>
      constraintSetup {angularLimitsMax=<0 5 5>, angularLimitsMin=<0 0 -5>, angularTarget=<1 0 0>}
      localJointOffset=<3 0 0>
     }]
    simSpaceSettings {maxAngularAcceleration=25, maxAngularVelocity=5, simSpaceAngularAlpha=1}
   }
   n20 = Anim Dynamics - Bone: dyn_jacketFlap_row_02_r!(ComponentPose=@n19, Alpha=1, GravityOverride:Vector="0,0,-980") -> (Pose) {
    angularDampingOverride=1
    angularSpringConstant=10000
    bAngularSpring=true
    bOverrideAngularDamping=true
    bUseGravityOverride=true
    gravityScale=0.35
    simulationSpace=RootRelative
    alphaBoolBlend {blendOption=Linear}
    boundBone {boneName=dyn_jacketFlap_row_02_r}
    componentAppliedLinearAccClamp=<5000 5000 5000>
    componentLinearAccScale=<1 1 1>
    componentLinearVelScale=<1 1 1>
    gravityOverride {z=-980}
    physicsBodyDefinitions [{
      boundBone {boneName=dyn_jacketFlap_row_02_r}
      boxExtents=<20 30 30>
      constraintSetup {angularLimitsMax=<0 5 10>, angularLimitsMin=<0 0 -10>, angularTarget=<1 0 0>}
      localJointOffset=<3 0 0>
     }]
    simSpaceSettings {maxAngularAcceleration=25, maxAngularVelocity=5, simSpaceAngularAlpha=1}
   }
   Save cached pose 'preFlapDynCache'(Pose=Component To Local(ComponentPose=@n20)) {cacheName=preFlapDynCache}
  ]}
 EventGraph {nodes [
   Event BlueprintUpdateAnimation()
   > n1 = Sequence!(execute) -> (then_0, then_1, then_2:Exec)
   Set faceAcc_mesh(
    execute=@n1.then_2
    faceAcc_mesh:Skeletal Mesh Component Budgeted Object Reference=Get Array Item with EFort Custom Part Type(Array=Event OnCharacterCustomizationLoaded().CustomMeshParts, PartType=Face)
   ) -> (Output_Get:Skeletal Mesh Component Budgeted Object Reference)
   n2 = Get faceAcc_mesh() -> (faceAcc_mesh:Skeletal Mesh Component Budgeted Object Reference)
   > Set hasVagabondFaceAcc(hasVagabondFaceAcc:bool=MatchesWildcard(SourceString=GetDisplayName(Object=@n2), Wildcard=*M_MED_Jonesy_Vagabond, SearchCase=CaseSensitive)) -> (Output_Get:bool)
   n3 = "Gravity Override State Function (Parameter Asset) (Nativized)"(
    execute=@n1.then_0
    CharacterPartAnimInstance=Self-Reference()
    GameTimeInSeconds=GetGameTimeInSeconds()
    Parameters=&/M_MED_RoyalWonk_Parameters
   )
   > Set Gravity Override(Gravity Override=@n3)
   Set isWearingCape(execute=@n1.then_1, isWearingCape=CheckForBackbling().bIsBackblingACape)
   > Set isCapeCrouching(isCapeCrouching:bool=AND Boolean(A=Get isWearingCape() -> (isWearingCape:bool), B=Get bIsCrouching())) -> (Output_Get:bool)
   n4 = OR Boolean(A=Get bIsCrouching(), B=Get bIsPlayingEmote())
   > n5 = Set RigidBodyFixNeeded(RigidBodyFixNeeded:bool=AND Boolean(A=Get bRigidBodyEnabled(), B=NOT Boolean(A=@n4))) -> (Output_Get:bool)
   > Set FERigidBodyFix(FERigidBodyFix:bool=AND Boolean(A=@n5, B=Get bIsInFrontEnd())) -> (Output_Get:bool)
  ]}
]
variables [
 faceAcc_mesh {type=SkeletalMeshComponentBudgeted}
 FERigidBodyFix {type=boolean}
 Gravity Override {type=Vector}
 Gravity Override Params {
  type=FortGravityOverrideParameterSet
  default=(Skydive_GoingUp=(GravityOverride=(WindFrequency=(X=1.500000,Y=7.000000,Z=8.000000),WindAmplitude=(X=200.000000,Y=-500.000000,Z=400.000000),GravityOverride=(X=0.000000,Y=-300.000000,Z=1500.000000),JointName="root")),Skydive_FloatingDown=(GravityOverride=(WindFrequency=(X=1.500000,Y=7.000000,Z=8.000000),WindAmplitude=(X=200.000000,Y=-500.000000,Z=400.000000),WindOffset=(X=0.000000,Y=-600.000000,Z=400.000000),GravityOverride=(X=0.000000,Y=-700.000000,Z=600.000000),JointName="root")),Skydive_DivingDown=(GravityOverride=(WindFrequency=(X=1.500000,Y=12.000000,Z=10.000000),WindAmplitude=(X=50.000000,Y=300.000000,Z=100.000000),GravityOverride=(X=-1500.000000,Y=0.000000,Z=0.000000),JointName="pelvis")),Skydive_Parachuting=(GravityOverride=(WindFrequency=(X=1.500000,Y=7.000000,Z=8.000000),WindAmplitude=(X=200.000000,Y=-500.000000,Z=400.000000),GravityOverride=(X=0.000000,Y=-800.000000,Z=-980.000000),JointName="root")),OnGround_Standing=(GravityOverride=(GravityOverride=(X=0.000000,Y=0.000000,Z=-980.000000),JointName="root")),OnGround_Moving=(GravityOverride=(GravityOverride=(X=0.000000,Y=0.000000,Z=-980.000000),JointName="root")),OnGround_Moving_HighRotationRate=(GravityOverride=(GravityOverride=(X=0.000000,Y=0.000000,Z=-980.000000),JointName="root")),OnGround_Crouch_Moving=(GravityOverride=(GravityOverride=(X=0.000000,Y=0.000000,Z=-980.000000),JointName="root")),Emote_or_Melee=(GravityOverride=(GravityOverride=(X=0.000000,Y=0.000000,Z=-980.000000),JointName="root")),Swimming=(GravityOverride=(GravityOverride=(X=-800.000000,Y=0.000000,Z=0.000000),JointName="pelvis")),Front_End=(GravityOverride=(GravityOverride=(X=0.000000,Y=0.000000,Z=-980.000000),JointName="root")),DBNO=(GravityOverride=(GravityOverride=(X=-980.000000,Y=400.000000,Z=0.000000),JointName="pelvis")),DBNO_Carried=(GravityOverride=(GravityOverride=(X=-980.000000,Y=500.000000,Z=-500.000000),JointName="pelvis")),OnGround_Falling=(GravityOverride=(GravityOverride=(X=0.000000,Y=0.000000,Z=-980.000000),JointName="root")),Ziplining=(GravityOverride=(GravityOverride=(X=0.000000,Y=0.000000,Z=-980.000000),JointName="root")),Skydive_Surfing=(GravityOverride=(WindFrequency=(X=1.500000,Y=7.000000,Z=8.000000),WindAmplitude=(X=200.000000,Y=-500.000000,Z=400.000000),GravityOverride=(X=500.000000,Y=-300.000000,Z=-980.000000),JointName="root")),Riding_Standing=(GravityOverride=(GravityOverride=(X=0.000000,Y=-350.000000,Z=-980.000000),JointName="root")),Riding_Moving=(GravityOverride=(GravityOverride=(X=0.000000,Y=-450.000000,Z=-450.000000),JointName="root")),Riding_Moving_Targeting=(GravityOverride=(GravityOverride=(X=0.000000,Y=-200.000000,Z=-980.000000),JointName="root")),Tactical_Sprint=(GravityOverride=(GravityOverride=(X=0.000000,Y=-450.000000,Z=-500.000000),JointName="root")),Rail_Grinding=(GravityOverride=(GravityOverride=(X=0.000000,Y=0.000000,Z=-980.000000),JointName="root")),)
 }
 hasVagabondFaceAcc {type=boolean}
 isCapeCrouching {type=boolean}
 isWearingCape {type=boolean}
 RigidBodyFixNeeded {type=boolean}
]
