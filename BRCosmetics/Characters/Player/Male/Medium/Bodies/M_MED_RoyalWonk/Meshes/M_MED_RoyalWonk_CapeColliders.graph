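# Physics asset of cape collision shapes for the M_MED_RoyalWonk male medium player body, going by its name.
# Capsules cover pelvis, spine, neck and clavicles, with overlapping capsules on spine_02 and spine_05 and a sphere on the head.
# Arm, leg, foot and suit back-flap bodies carry no shapes, and there are no constraints, so it describes upper-body collision volumes rather than a jointed ragdoll.
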PhysicsAsset /BRCosmetics/Characters/Player/Male/Medium/Bodies/M_MED_RoyalWonk/Meshes/M_MED_RoyalWonk_CapeColliders
bodies {
 calf_l []
 calf_r []
 clavicle_l [{shapeType=Capsule, length=9.0192, radius=10.8892, center=<9.4266 -1.1806 -2.4977>, rotation=<rot 69.3297 1.4474 170.4283>}]
 clavicle_r [{shapeType=Capsule, length=9.0192, radius=10.8892, center=<-9.4266 1.1806 2.4977>, rotation=<rot 69.3297 1.4474 -9.5717>}]
 dyn_SuitBackFlap_02_l []
 dyn_SuitBackFlap_02_r []
 foot_l []
 foot_r []
 head [{shapeType=Sphere, radius=19.072, center=<9.1193 1.7143 0>}]
 lowerarm_l []
 lowerarm_r []
 neck_02 [{shapeType=Capsule, length=9.6832, radius=9.0637, center=<-0.3598 0.9372 0>, rotation=<rot 90 4.2785 0>}]
 pelvis [{shapeType=Capsule, length=9.2712, radius=21.1827, center=<-2.4887 0.1881 0>}]
 spine_01 [{shapeType=Capsule, length=9.3673, radius=18.7791, center=<-1.7965 0.1884 0>}]
 spine_02 [
  {shapeType=Capsule, length=10.5034, radius=14.3705, center=<-1.711 -0.1851 0.0001>}
  {shapeType=Capsule, length=29.3956, radius=11.5807, center=<2.5024 0.4992 3.9825>, rotation=<rot 90 0.0049 0.0049>}
  {shapeType=Capsule, length=29.875, radius=11.1013, center=<2.4704 0.4991 -3.9825>, rotation=<rot 90 -0.0396 -0.0396>}
 ]
 spine_03 [{shapeType=Capsule, length=8.9614, radius=13.7031, center=<0.1445 0.2396 0.0001>}]
 spine_04 [{shapeType=Capsule, length=9.4404, radius=14.1325, center=<2.766 0.7765 0.0002>}]
 spine_05 [
  {shapeType=Capsule, length=8.0226, radius=15.695, center=<-2.2795 1.8109 0>}
  {shapeType=Capsule, length=13.9518, radius=13.9425, center=<-0.2916 0.4452 0>}
 ]
 thigh_l []
 thigh_r []
 upperarm_l []
 upperarm_r []
}
constraints []
